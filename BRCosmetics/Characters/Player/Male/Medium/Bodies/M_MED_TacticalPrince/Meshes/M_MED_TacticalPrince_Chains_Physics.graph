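# Physics asset for the dangling chains on the TacticalPrince medium male player body cosmetic.
# Front and back chains hang from spine_01, and in and out chains from the left clavicle, via small base anchor bodies.
# Dyn chain links are thin capsules joined by constraints with tight swing limits and zero-degree twist limits.
# Larger capsules approximate the pelvis, spine and left clavicle, upper arm and thigh, joined by free-rotation constraints.
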PhysicsAsset /BRCosmetics/Characters/Player/Male/Medium/Bodies/M_MED_TacticalPrince/Meshes/M_MED_TacticalPrince_Chains_Physics
bodies {
 base_backChain_01 [{shapeType=Capsule, length=0.505, radius=0.505, shapeName=base_backChain_01_capsule, center=<0 0 0.1405>, rotation=<rot 0 -90 0>}]
 base_frontChain_01 [{shapeType=Capsule, length=0.505, radius=0.505, shapeName=base_frontChain_01_capsule, rotation=<rot 0 -90 0>}]
 base_inChain_01 [{shapeType=Capsule, length=0.505, radius=0.505, shapeName=base_inChain_01_capsule, rotation=<rot 0 -90 0>}]
 base_outChain_01 [{shapeType=Capsule, length=0.505, radius=0.505, shapeName=base_outChain_01_capsule, rotation=<rot 0 -90 0>}]
 clavicle_l [{
   shapeType=Capsule
   length=10.8974
   radius=7.2082
   shapeName=clavicle_l_capsule
   center=<9.5013 -1.2038 1.3058>
   rotation=<rot -80 179.9992 -179.9993>
  }]
 dyn_backChain_01 [{
   shapeType=Capsule
   length=4.8082
   radius=0.6961
   shapeName=dyn_backChain_01_capsule
   center=<2.5983 -0.6227 0.3465>
   rotation=<rot 0.0597 80.4801 -83.9294>
  }]
 dyn_backChain_02 [{
   shapeType=Capsule
   length=4.2037
   radius=0.7413
   shapeName=dyn_backChain_02_capsule
   center=<1.6719 0.0539 -0.128>
   rotation=<rot 13.2251 94.6739 -88.6818>
  }]
 dyn_frontChain_01 [{
   shapeType=Capsule
   length=4.5728
   radius=0.6567
   shapeName=dyn_frontChain_01_capsule
   center=<2.5902 -0.1356 0.0102>
   rotation=<rot 0 -89.6398 -89.099>
  }]
 dyn_frontChain_02 [{
   shapeType=Capsule
   length=4.9846
   radius=0.6078
   shapeName=dyn_frontChain_02_capsule
   center=<2.7348 -0.0499 0.1451>
   rotation=<rot 0 90.3699 -87.3448>
  }]
 dyn_frontChain_03 [{
   shapeType=Capsule
   length=3.9264
   radius=0.7577
   shapeName=dyn_frontChain_03_capsule
   center=<1.5389 -0.0934 -0.0917>
   rotation=<rot 0.0001 85.3351 -95.6894>
  }]
 dyn_inChain_01 [{
   shapeType=Capsule
   length=4.4384
   radius=0.5839
   shapeName=dyn_inChain_01_capsule
   center=<1.7642 -0.39 -0.1174>
   rotation=<rot 0.4449 -96.4885 -84.907>
  }]
 dyn_inChain_02 [{
   shapeType=Capsule
   length=4.4931
   radius=0.5
   shapeName=dyn_inChain_02_capsule
   center=<2.0482 -0.1771 -0.1667>
   rotation=<rot -0.5316 95.0823 -88.9104>
  }]
 dyn_inChain_03 [{
   shapeType=Capsule
   length=4.2804
   radius=0.6195
   shapeName=dyn_inChain_03_capsule
   center=<1.4662 0.0828 -0.0013>
   rotation=<rot -0.3956 87.5064 -85.7717>
  }]
 dyn_outChain_01 [{
   shapeType=Capsule
   length=5.1383
   radius=0.537
   shapeName=dyn_outChain_01_capsule
   center=<2.7955 0.1586 -0.021>
   rotation=<rot 0 -86.2075 -90.9667>
  }]
 dyn_outChain_02 [{
   shapeType=Capsule
   length=4.8089
   radius=0.6642
   shapeName=dyn_outChain_02_capsule
   center=<2.1752 0.0349 -0.0209>
   rotation=<rot 0 86.687 -89.6862>
  }]
 pelvis [{
   shapeType=Capsule
   length=12.3257
   radius=10.2372
   shapeName=pelvis_capsule
   center=<-2.4457 -1.7776 -0.0036>
   rotation=<rot -2.0022 -90 0.7865>
  }]
 spine_01 [{shapeType=Capsule, length=9.0538, radius=9.7691, shapeName=spine_01_capsule, center=<2.4645 0.0857 0>, rotation=<rot 0 -90 0>}]
 spine_02 [{
   shapeType=Capsule
   length=6.9635
   radius=10.2252
   shapeName=spine_02_capsule
   center=<3.9092 1.0337 0.0086>
   rotation=<rot -0.0867 -90 0.0267>
  }]
 spine_03 [{shapeType=Capsule, length=5.9838, radius=13.1719, shapeName=spine_03_capsule, center=<10.3468 1.2263 0>, rotation=<rot 1.3906 -90 -0.3178>}]
 spine_04 [{
   shapeType=Capsule
   length=12.7137
   radius=11.5224
   shapeName=spine_04_capsule
   center=<6.7047 0.6996 0.0036>
   rotation=<rot -0.3302 -90 0.2757>
  }]
 spine_05 [{shapeType=Capsule, length=17.2433, radius=9.9178, shapeName=spine_05_capsule, center=<2.7595 -1.6661 0.4414>, rotation=<rot 0 -90 0>}]
 thigh_l [{
   shapeType=Capsule
   length=20.9377
   radius=7.6482
   shapeName=thigh_l_capsule
   center=<-15.0146 -0.024 0.2027>
   rotation=<rot 1.3606 93.747 -88.9201>
  }]
 upperarm_l [{shapeType=Capsule, length=13.1602, radius=7.7376, shapeName=upperarm_l_capsule, center=<8.4945 -0.5044 -0.7828>, rotation=<rot -90 180 180>}]
}
constraints [
 {bone1Name=dyn_frontChain_02, bone2Name=dyn_frontChain_01, swing1LimitDegrees=15, swing2LimitDegrees=10, twistLimitDegrees=0}
 {bone1Name=dyn_frontChain_03, bone2Name=dyn_frontChain_02, swing1LimitDegrees=15, swing2LimitDegrees=10, twistLimitDegrees=0}
 {bone1Name=dyn_frontChain_01, bone2Name=base_frontChain_01, swing1LimitDegrees=15, swing2LimitDegrees=10, twistLimitDegrees=0}
 {bone1Name=dyn_backChain_01, bone2Name=base_backChain_01, swing1LimitDegrees=15, swing2LimitDegrees=10, twistLimitDegrees=0}
 {bone1Name=dyn_backChain_02, bone2Name=dyn_backChain_01, swing1LimitDegrees=15, swing2LimitDegrees=10, twistLimitDegrees=0}
 {bone1Name=dyn_inChain_01, bone2Name=base_inChain_01, swing1LimitDegrees=25, swing2LimitDegrees=7.5, twistLimitDegrees=0}
 {bone1Name=dyn_inChain_02, bone2Name=dyn_inChain_01, swing1LimitDegrees=25, swing2LimitDegrees=7.5, twistLimitDegrees=0}
 {bone1Name=dyn_inChain_03, bone2Name=dyn_inChain_02, swing1LimitDegrees=25, swing2LimitDegrees=7.5, twistLimitDegrees=0}
 {bone1Name=dyn_outChain_01, bone2Name=base_outChain_01, swing1LimitDegrees=25, swing2LimitDegrees=7.5, twistLimitDegrees=0}
 {bone1Name=dyn_outChain_02, bone2Name=dyn_outChain_01, swing1LimitDegrees=25, swing2LimitDegrees=7.5, twistLimitDegrees=0}
 {
  bone1Name=base_outChain_01
  bone2Name=clavicle_l
  swing1LimitDegrees=45
  swing1Motion=Free
  swing2LimitDegrees=45
  swing2Motion=Free
  twistLimitDegrees=45
  twistMotion=Free
 }
 {
  bone1Name=base_inChain_01
  bone2Name=clavicle_l
  swing1LimitDegrees=45
  swing1Motion=Free
  swing2LimitDegrees=45
  swing2Motion=Free
  twistLimitDegrees=45
  twistMotion=Free
 }
 {
  bone1Name=base_frontChain_01
  bone2Name=spine_01
  swing1LimitDegrees=45
  swing1Motion=Free
  swing2LimitDegrees=45
  swing2Motion=Free
  twistLimitDegrees=45
  twistMotion=Free
 }
 {
  bone1Name=base_backChain_01
  bone2Name=spine_01
  swing1LimitDegrees=45
  swing1Motion=Free
  swing2LimitDegrees=45
  swing2Motion=Free
  twistLimitDegrees=45
  twistMotion=Free
 }
 {
  bone1Name=spine_01
  bone2Name=pelvis
  swing1LimitDegrees=45
  swing1Motion=Free
  swing2LimitDegrees=45
  swing2Motion=Free
  twistLimitDegrees=45
  twistMotion=Free
 }
 {
  bone1Name=thigh_l
  bone2Name=pelvis
  swing1LimitDegrees=45
  swing1Motion=Free
  swing2LimitDegrees=45
  swing2Motion=Free
  twistLimitDegrees=45
  twistMotion=Free
 }
 {
  bone1Name=spine_05
  bone2Name=spine_04
  swing1LimitDegrees=45
  swing1Motion=Free
  swing2LimitDegrees=45
  swing2Motion=Free
  twistLimitDegrees=45
  twistMotion=Free
 }
 {
  bone1Name=spine_04
  bone2Name=spine_03
  swing1LimitDegrees=45
  swing1Motion=Free
  swing2LimitDegrees=45
  swing2Motion=Free
  twistLimitDegrees=45
  twistMotion=Free
 }
 {
  bone1Name=spine_03
  bone2Name=spine_02
  swing1LimitDegrees=45
  swing1Motion=Free
  swing2LimitDegrees=45
  swing2Motion=Free
  twistLimitDegrees=45
  twistMotion=Free
 }
 {
  bone1Name=clavicle_l
  bone2Name=spine_05
  swing1LimitDegrees=45
  swing1Motion=Free
  swing2LimitDegrees=45
  swing2Motion=Free
  twistLimitDegrees=45
  twistMotion=Free
 }
 {
  bone1Name=upperarm_l
  bone2Name=clavicle_l
  swing1LimitDegrees=45
  swing1Motion=Free
  swing2LimitDegrees=45
  swing2Motion=Free
  twistLimitDegrees=45
  twistMotion=Free
 }
 {
  bone1Name=spine_02
  bone2Name=spine_01
  swing1LimitDegrees=45
  swing1Motion=Free
  swing2LimitDegrees=45
  swing2Motion=Free
  twistLimitDegrees=45
  twistMotion=Free
 }
]
