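# Physics asset for the jacket and neck lapels of the male medium TacticalPrince body, setting up their jointed secondary motion.
# Large capsules cover upper spine, neck, head, clavicles and upper arms, linked by free-swinging, free-twisting constraints.
# Tiny base-bone capsules attach freely to spine_04/05, each anchoring a dyn lapel bone with limited, mostly hinge-like swing tuned per side.
# Jacket lapels use two or three side-by-side capsules to approximate flat flaps.
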PhysicsAsset /BRCosmetics/Characters/Player/Male/Medium/Bodies/M_MED_TacticalPrince/Meshes/M_MED_TacticalPrince_Lapels_Physics
bodies {
 base_jacktLapel_l [{shapeType=Capsule, length=0.505, radius=0.505, shapeName=base_jacktLapel_l_capsule, rotation=<rot 0 -90 0>}]
 base_jacktLapel_low_l [{shapeType=Capsule, length=0.505, radius=0.505, shapeName=base_jacktLapel_low_l_capsule, rotation=<rot 0 -90 0>}]
 base_jacktLapel_low_r [{shapeType=Capsule, length=0.505, radius=0.505, shapeName=base_jacktLapel_low_r_capsule, rotation=<rot 0 -90 0>}]
 base_jacktLapel_r [{shapeType=Capsule, length=0.505, radius=0.505, shapeName=base_jacktLapel_r_capsule, rotation=<rot 0 -90 0>}]
 base_neckLapel_01_l [{shapeType=Capsule, length=0.505, radius=0.505, shapeName=base_neckLapel_01_l_capsule, rotation=<rot 0 -90 0>}]
 base_neckLapel_01_r [{shapeType=Capsule, length=0.505, radius=0.505, shapeName=base_neckLapel_01_r_capsule, rotation=<rot 0 -90 0>}]
 base_neckLapel_02_l [{shapeType=Capsule, length=0.505, radius=0.505, shapeName=base_neckLapel_02_l_capsule, rotation=<rot 0 -90 0>}]
 base_neckLapel_02_r [{shapeType=Capsule, length=0.505, radius=0.505, shapeName=base_neckLapel_02_r_capsule, rotation=<rot 0 -90 0>}]
 base_neckLapel_b [{shapeType=Capsule, length=0.505, radius=0.505, shapeName=base_neckLapel_b_capsule, rotation=<rot 0 -90 0>}]
 clavicle_l [{
   shapeType=Capsule
   length=8.2066
   radius=8.9579
   shapeName=clavicle_l_capsule
   center=<7.5399 -1.4286 -0.2615>
   rotation=<rot 1.4472 88.5801 -106.0454>
  }]
 clavicle_r [{
   shapeType=Capsule
   length=8.2066
   radius=8.9579
   shapeName=clavicle_r_capsule
   center=<-7.5399 1.4286 0.2615>
   rotation=<rot 1.4472 88.5801 73.9546>
  }]
 dyn_jacktLapel_l [
  {
   shapeType=Capsule
   length=4.9715
   radius=0.963
   shapeName=dyn_jacktLapel_l_capsule
   center=<7.4795 -0.1734 -1.2169>
   rotation=<rot -0.0004 -90.721 5.2135>
  }
  {
   shapeType=Capsule
   length=13.4713
   radius=1.006
   shapeName=dyn_jacktLapel_l_capsule
   center=<5.2937 -0.399 -1.7032>
   rotation=<rot -0.0004 -90.721 5.2135>
  }
  {
   shapeType=Capsule
   length=10.1827
   radius=1.2635
   shapeName=dyn_jacktLapel_l_capsule
   center=<2.4208 -0.9104 -1.1251>
   rotation=<rot -0.0005 -90.721 3.2731>
  }
 ]
 dyn_jacktLapel_low_l [
  {
   shapeType=Capsule
   length=11.9879
   radius=1.3062
   shapeName=dyn_jacktLapel_low_l_capsule
   center=<1.3537 -0.7722 0.6389>
   rotation=<rot 8.2812 -88.915 4.1866>
  }
  {
   shapeType=Capsule
   length=10.8795
   radius=1.3062
   shapeName=dyn_jacktLapel_low_l_capsule
   center=<4.0612 -0.4946 0.4503>
   rotation=<rot 7.5885 -88.5846 -9.5308>
  }
 ]
 dyn_jacktLapel_low_r [
  {
   shapeType=Capsule
   length=12.379
   radius=1.1435
   shapeName=dyn_jacktLapel_low_r_capsule
   center=<4.6451 -1.0403 1.8391>
   rotation=<rot 0.1409 -89.9938 -178.6319>
  }
  {
   shapeType=Capsule
   length=12.0111
   radius=1.1111
   shapeName=dyn_jacktLapel_low_r_capsule
   center=<1.6293 -0.9077 1.7026>
   rotation=<rot -3.6614 -90.9122 -174.1484>
  }
 ]
 dyn_jacktLapel_r [
  {
   shapeType=Capsule
   length=10.3434
   radius=1.4618
   shapeName=dyn_jacktLapel_r_capsule
   center=<2.0645 -1.0441 2.0644>
   rotation=<rot -0.0002 89.279 -0.8175>
  }
  {
   shapeType=Capsule
   length=9.831
   radius=1.2778
   shapeName=dyn_jacktLapel_r_capsule
   center=<4.863 -0.939 2.0108>
   rotation=<rot -0.0003 89.279 -0.0021>
  }
  {
   shapeType=Capsule
   length=5.8238
   radius=1.2778
   shapeName=dyn_jacktLapel_r_capsule
   center=<7.437 -0.5238 1.676>
   rotation=<rot -0.0003 89.279 -0.0021>
  }
 ]
 dyn_neckLapel_01_l [{
   shapeType=Capsule
   length=7.2442
   radius=0.9659
   shapeName=dyn_neckLapel_01_l_capsule
   center=<5.8279 -0.6149 -2.373>
   rotation=<rot 2.1088 -92.0228 -156.344>
  }]
 dyn_neckLapel_01_r [{
   shapeType=Capsule
   length=7.1769
   radius=1.0218
   shapeName=dyn_neckLapel_01_r_capsule
   center=<5.9739 -0.2589 1.8986>
   rotation=<rot -6.0598 -88.9935 -25.576>
  }]
 dyn_neckLapel_02_l [{
   shapeType=Capsule
   length=8.9495
   radius=1.1746
   shapeName=dyn_neckLapel_01_l_capsule
   center=<4.8564 -0.6779 -1.238>
   rotation=<rot 15.355 -93.0197 179.4393>
  }]
 dyn_neckLapel_02_r [{
   shapeType=Capsule
   length=6.8868
   radius=1.1325
   shapeName=dyn_neckLapel_01_l_capsule
   center=<4.6423 -0.591 1.1882>
   rotation=<rot -15.5449 -93.9501 -175.3139>
  }]
 dyn_neckLapel_b [{shapeType=Capsule, length=9.8232, radius=1.5358, shapeName=dyn_neckLapel_01_l_capsule, center=<4.6421 0.4942 0>, rotation=<rot 0 -90 0>}]
 head [{shapeType=Capsule, length=1.4744, radius=5.6508, shapeName=head_capsule, center=<2.4265 0.533 0>, rotation=<rot 0 -95.0003 0>}]
 neck_01 [{shapeType=Capsule, length=1.1184, radius=6.0179, shapeName=neck_01_capsule, center=<1.3216 0.7128 0>, rotation=<rot 0 -90 0>}]
 neck_02 [{shapeType=Capsule, length=0, radius=5.8398, shapeName=neck_02_capsule, center=<2.2469 0.701 0>, rotation=<rot 0 -90 0>}]
 spine_03 [{shapeType=Capsule, length=8.5521, radius=12.8002, shapeName=spine_03_capsule, center=<5.1687 1.354 0>}]
 spine_04 [{shapeType=Capsule, length=10.0188, radius=12.8529, shapeName=spine_04_capsule, center=<2.4309 0.4098 -0.2417>}]
 spine_05 [{shapeType=Capsule, length=14.2819, radius=12.614, shapeName=spine_05_capsule, center=<-2.5302 -1.2528 0>}]
 upperarm_l [{
   shapeType=Capsule
   length=18.3131
   radius=6.5769
   shapeName=upperarm_l_capsule
   center=<11.8025 -0.2542 -1.1789>
   rotation=<rot 88.3937 -3.4155 -3.5047>
  }]
 upperarm_r [{
   shapeType=Capsule
   length=18.3131
   radius=6.5769
   shapeName=upperarm_r_capsule
   center=<-11.775 -0.2335 0.5485>
   rotation=<rot 89.5742 -12.5966 167.3195>
  }]
}
constraints [
 {
  bone1Name=spine_05
  bone2Name=spine_04
  swing1LimitDegrees=45
  swing1Motion=Free
  swing2LimitDegrees=45
  swing2Motion=Free
  twistLimitDegrees=45
  twistMotion=Free
 }
 {
  bone1Name=clavicle_l
  bone2Name=spine_05
  swing1LimitDegrees=45
  swing1Motion=Free
  swing2LimitDegrees=45
  swing2Motion=Free
  twistLimitDegrees=45
  twistMotion=Free
 }
 {
  bone1Name=clavicle_r
  bone2Name=spine_05
  swing1LimitDegrees=45
  swing1Motion=Free
  swing2LimitDegrees=45
  swing2Motion=Free
  twistLimitDegrees=45
  twistMotion=Free
 }
 {
  bone1Name=upperarm_l
  bone2Name=clavicle_l
  swing1LimitDegrees=45
  swing1Motion=Free
  swing2LimitDegrees=45
  swing2Motion=Free
  twistLimitDegrees=45
  twistMotion=Free
 }
 {
  bone1Name=upperarm_r
  bone2Name=clavicle_r
  swing1LimitDegrees=45
  swing1Motion=Free
  swing2LimitDegrees=45
  swing2Motion=Free
  twistLimitDegrees=45
  twistMotion=Free
 }
 {bone1Name=dyn_jacktLapel_low_l, bone2Name=base_jacktLapel_low_l, swing1LimitDegrees=140, swing2LimitDegrees=20, twistLimitDegrees=20}
 {bone1Name=dyn_jacktLapel_low_r, bone2Name=base_jacktLapel_low_r, swing1LimitDegrees=140, swing2LimitDegrees=15, twistLimitDegrees=15}
 {
  bone1Name=base_jacktLapel_l
  bone2Name=spine_05
  swing1LimitDegrees=45
  swing1Motion=Free
  swing2LimitDegrees=45
  swing2Motion=Free
  twistLimitDegrees=45
  twistMotion=Free
 }
 {
  bone1Name=base_jacktLapel_r
  bone2Name=spine_05
  swing1LimitDegrees=45
  swing1Motion=Free
  swing2LimitDegrees=45
  swing2Motion=Free
  twistLimitDegrees=45
  twistMotion=Free
 }
 {bone1Name=dyn_jacktLapel_r, bone2Name=base_jacktLapel_r, swing1LimitDegrees=140, swing2LimitDegrees=10, twistLimitDegrees=5}
 {bone1Name=dyn_jacktLapel_l, bone2Name=base_jacktLapel_l, swing1LimitDegrees=140, swing2LimitDegrees=10, twistLimitDegrees=15}
 {bone1Name=dyn_neckLapel_01_l, bone2Name=base_neckLapel_01_l, swing1LimitDegrees=140, swing2LimitDegrees=20, twistLimitDegrees=25}
 {bone1Name=dyn_neckLapel_01_r, bone2Name=base_neckLapel_01_r, swing1LimitDegrees=140, swing2LimitDegrees=10, twistLimitDegrees=5}
 {
  bone1Name=base_jacktLapel_low_l
  bone2Name=spine_04
  swing1LimitDegrees=45
  swing1Motion=Free
  swing2LimitDegrees=45
  swing2Motion=Free
  twistLimitDegrees=45
  twistMotion=Free
 }
 {
  bone1Name=base_jacktLapel_low_r
  bone2Name=spine_04
  swing1LimitDegrees=45
  swing1Motion=Free
  swing2LimitDegrees=45
  swing2Motion=Free
  twistLimitDegrees=45
  twistMotion=Free
 }
 {
  bone1Name=base_neckLapel_01_r
  bone2Name=spine_05
  swing1LimitDegrees=45
  swing1Motion=Free
  swing2LimitDegrees=45
  swing2Motion=Free
  twistLimitDegrees=45
  twistMotion=Free
 }
 {
  bone1Name=base_neckLapel_01_l
  bone2Name=spine_05
  swing1LimitDegrees=45
  swing1Motion=Free
  swing2LimitDegrees=45
  swing2Motion=Free
  twistLimitDegrees=45
  twistMotion=Free
 }
 {
  bone1Name=neck_01
  bone2Name=spine_05
  swing1LimitDegrees=45
  swing1Motion=Free
  swing2LimitDegrees=45
  swing2Motion=Free
  twistLimitDegrees=45
  twistMotion=Free
 }
 {
  bone1Name=neck_02
  bone2Name=neck_01
  swing1LimitDegrees=45
  swing1Motion=Free
  swing2LimitDegrees=45
  swing2Motion=Free
  twistLimitDegrees=45
  twistMotion=Free
 }
 {
  bone1Name=head
  bone2Name=neck_02
  swing1LimitDegrees=45
  swing1Motion=Free
  swing2LimitDegrees=45
  swing2Motion=Free
  twistLimitDegrees=45
  twistMotion=Free
 }
 {bone1Name=dyn_neckLapel_02_l, bone2Name=base_neckLapel_02_l, swing1LimitDegrees=45, swing2LimitDegrees=10, twistLimitDegrees=15}
 {bone1Name=dyn_neckLapel_02_r, bone2Name=base_neckLapel_02_r, swing1LimitDegrees=45, swing2LimitDegrees=10, twistLimitDegrees=10}
 {bone1Name=dyn_neckLapel_b, bone2Name=base_neckLapel_b, swing1LimitDegrees=90, swing2LimitDegrees=10, twistLimitDegrees=10}
 {
  bone1Name=base_neckLapel_b
  bone2Name=spine_05
  swing1LimitDegrees=45
  swing1Motion=Free
  swing2LimitDegrees=45
  swing2Motion=Free
  twistLimitDegrees=45
  twistMotion=Free
 }
 {
  bone1Name=base_neckLapel_02_r
  bone2Name=spine_05
  swing1LimitDegrees=45
  swing1Motion=Free
  swing2LimitDegrees=45
  swing2Motion=Free
  twistLimitDegrees=45
  twistMotion=Free
 }
 {
  bone1Name=base_neckLapel_02_l
  bone2Name=spine_05
  swing1LimitDegrees=45
  swing1Motion=Free
  swing2LimitDegrees=45
  swing2Motion=Free
  twistLimitDegrees=45
  twistMotion=Free
 }
 {
  bone1Name=spine_04
  bone2Name=spine_03
  swing1LimitDegrees=45
  swing1Motion=Free
  swing2LimitDegrees=45
  swing2Motion=Free
  twistLimitDegrees=45
  twistMotion=Free
 }
]
