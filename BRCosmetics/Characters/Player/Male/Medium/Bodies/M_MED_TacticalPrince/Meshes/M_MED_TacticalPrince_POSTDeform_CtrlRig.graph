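# Post-deform Control Rig for the M_MED_TacticalPrince player body that corrects its dangling cosmetic chain bones.
# Position constraints with maintained offset tie the in, out, front and back chain roots to PosRef nulls.
# The in/out nulls sit under base bones, while front/back nulls are children of the bones they constrain.
# Pub_ISCelebreteEmote, smoothed by AlphaInterp, then weights a rotation tweak on dyn_outChain_01, presumably for celebration emotes.
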ControlRigBlueprint /BRCosmetics/Characters/Player/Male/Medium/Bodies/M_MED_TacticalPrince/Meshes/M_MED_TacticalPrince_POSTDeform_CtrlRig
graphs [
 RigVMFunctionLibrary {links [], nodes []}
 RigVMModel {nodes [
   RigUnit_BeginExecution = &RigUnit_BeginExecution()
   /Script/ControlRig.RigUnit_PositionConstraintLocalSpaceOffset(
    Child=(Type=Bone,Name="dyn_inChain_01")
    Filter=(bX=True,bY=True,bZ=True)
    Parents=((Item=(Type=Null,Name="dyn_inChain_01_PosRef_null"),Weight=1))
    Weight=1
    bMaintainOffset=True
    ExecutePin=@RigUnit_BeginExecution.ExecutePin
   ) {is=struct RigVMUnitNode}
   > &RigUnit_PositionConstraintLocalSpaceOffset(
    Child=(Type=Bone,Name="dyn_outChain_01")
    Filter=(bX=True,bY=True,bZ=True)
    Parents=((Item=(Type=Null,Name="dyn_outChain_01_PosRef_null"),Weight=1))
    Weight=1
    bMaintainOffset=True
   )
   > &RigUnit_PositionConstraintLocalSpaceOffset(
    Child=(Type=Bone,Name="dyn_frontChain_01")
    Filter=(bX=True,bY=True,bZ=True)
    Parents=((Item=(Type=Null,Name="dyn_frontChain_01_PosRef_null"),Weight=1))
    Weight=1
    bMaintainOffset=True
   )
   > &RigUnit_PositionConstraintLocalSpaceOffset(
    Child=(Type=Bone,Name="dyn_backChain_01")
    Filter=(bX=True,bY=True,bZ=True)
    Parents=((Item=(Type=Null,Name="dyn_BackChain_01_PosRef_null"),Weight=1))
    Weight=1
    bMaintainOffset=True
   )
   AlphaInterp = &RigVMFunction_AlphaInterp(
    InterpSpeedDecreasing=5
    InterpSpeedIncreasing=5
    Scale=1
    bClampResult=False
    Value=RigVMVariableNode() {variable=Pub_ISCelebreteEmote}.Value
   )
   > &RigUnit_ModifyTransforms(
    ItemToModify=((Item=(Type=Bone,Name="dyn_outChain_01"),Transform=(Rotation=(X=-0.0436,Y=0.0755,Z=0.4981,W=0.8627),Translation=(X=0,Y=0,Z=0),Scale3D=(X=1,Y=1,Z=1))))
    WeightMinimum=0
    Weight=@AlphaInterp.Result
   )
  ]}
]
hierarchy [
 root {loc=<0 0 0>, rot=<0 0 0 1>}
 pelvis {parent=root, loc=<0 -0.7071 93.1191>, rot=<0.0127 0.707 -0.0127 -0.707>}
 spine_01 {parent=pelvis, loc=<6.7494 -0.6388 0>, rot=<0 0 0.034 -0.9994>}
 spine_02 {parent=spine_01, loc=<7.2518 -0.035 0>, rot=<0 0 0.0238 -0.9997>}
 spine_03 {parent=spine_02, loc=<6.0104 0 0>, rot=<0 0 0.0149 -0.9999>}
 spine_04 {parent=spine_03, loc=<8.1024 0 0>, rot=<0 0 -0.0132 -0.9999>}
 spine_05 {parent=spine_04, loc=<10.0146 0 0>, rot=<0 0 -0.0015 -1>}
 neck_01 {parent=spine_05, loc=<11.3138 0.1529 0>, rot=<0 0 -0.2003 -0.9797>}
 neck_02 {parent=neck_01, loc=<4.5627 0 0>, rot=<0 0 0.0504 -0.9987>}
 head {parent=neck_02, loc=<5.3914 0 0>, rot=<0 0 0.115 -0.9934>}
 attach_cape {parent=spine_05, loc=<9.017 -11.714 0>, rot=<0.0616 -0.7044 -0.0616 -0.7044>}
 cape {parent=attach_cape, type=Null, loc=<0 0 0>, rot=<0 0 0 1>}
 attach_backpack {parent=spine_05, loc=<-0.123 -13.473 0>, rot=<-0.0093 -0.707 0.0093 -0.707>}
 Backpack_BR {parent=attach_backpack, type=Null, loc=<0 0 0>, rot=<0 0 0 1>}
 clavicle_l {parent=spine_05, loc=<4.3575 -0.1466 -2.3223>, rot=<0.0465 -0.7289 -0.009 -0.6829>}
 upperarm_l {parent=clavicle_l, loc=<15.5111 0 0>, rot=<0.0226 -0.3234 -0.0088 -0.9459>}
 lowerarm_l {parent=upperarm_l, loc=<26.6514 0 0>, rot=<0.0391 -0.0099 -0.2453 -0.9686>}
 hand_l {parent=lowerarm_l, loc=<22.3491 0 0>, rot=<-0.7081 -0.0507 0.0975 -0.6975>}
 thumb_01_l {parent=hand_l, loc=<1.3118 -0.3191 -2.459>, rot=<0.6546 -0.4042 0.0664 -0.6354>}
 thumb_02_l {parent=thumb_01_l, loc=<5.9336 -0.0448 0.0401>, rot=<-0.0037 0.0213 0.298 -0.9543>}
 thumb_03_l {parent=thumb_02_l, loc=<2.2223 0.3682 -0.0907>, rot=<0.0118 -0.013 0.0066 -0.9998>}
 index_metacarpal_l {parent=hand_l, loc=<1.6535 0.0183 -2.96>, rot=<0.0158 -0.0124 -0.0173 -0.9996>}
 index_01_l {parent=index_metacarpal_l, loc=<8.3008 0.0357 -0.0524>, rot=<-0.0044 0.0485 0.0924 -0.9945>}
 index_02_l {parent=index_01_l, loc=<3.3587 -0.3004 0.0452>, rot=<-0.002 0.0027 0.3225 -0.9466>}
 index_03_l {parent=index_02_l, loc=<2.5578 -0.123 -0.0252>, rot=<-0.002 0.0067 0.2558 -0.9667>}
 middle_metacarpal_l {parent=hand_l, loc=<2.1856 -0.1725 -0.5984>, rot=<0.006 0.0168 -0.0329 -0.9993>}
 middle_01_l {parent=middle_metacarpal_l, loc=<7.3275 -0.2447 0.0129>, rot=<-0.0208 0.0042 0.0858 -0.9961>}
 middle_02_l {parent=middle_01_l, loc=<4.0123 0.1401 0.1175>, rot=<-0.0218 0.0258 0.3512 -0.9357>}
 middle_03_l {parent=middle_02_l, loc=<2.6077 0.0534 0.1373>, rot=<0.0082 0.0188 0.1708 -0.9851>}
 ring_metacarpal_l {parent=hand_l, loc=<2.2074 -0.3625 1.3638>, rot=<-0.0284 0.0722 -0.0458 -0.9959>}
 ring_01_l {parent=ring_metacarpal_l, loc=<7.0295 -0.6197 -0.1446>, rot=<-0.0489 -0.0252 0.1892 -0.9804>}
 ring_02_l {parent=ring_01_l, loc=<3.6174 0.4647 -0.1027>, rot=<0 0 0.2806 -0.9598>}
 ring_03_l {parent=ring_02_l, loc=<2.4152 0.114 0.0199>, rot=<0 0 0.1189 -0.9929>}
 pinky_metacarpal_l {parent=hand_l, loc=<1.7616 -0.9753 3.3814>, rot=<-0.0397 0.0814 -0.069 -0.9935>}
 pinky_01_l {parent=pinky_metacarpal_l, loc=<7.1709 -0.4831 -0.1094>, rot=<-0.0654 -0.0035 0.2242 -0.9723>}
 pinky_02_l {parent=pinky_01_l, loc=<2.8274 0.0893 -0.1043>, rot=<0.0037 0.0011 0.2854 -0.9584>}
 pinky_03_l {parent=pinky_02_l, loc=<2.2046 0.0307 -0.0078>, rot=<-0.0002 0.0098 0.1445 -0.9895>}
 deform_wrist_dn_base_l {parent=hand_l, loc=<0 -2.2342 0>, rot=<0 0 0 -1>}
 deform_wrist_dn_l {parent=deform_wrist_dn_base_l, loc=<0 0 0>, rot=<0 0 0 1>}
 deform_wrist_up_base_l {parent=hand_l, loc=<0 2.2682 0>, rot=<0 0 0 -1>}
 deform_wrist_up_l {parent=deform_wrist_up_base_l, loc=<0 0 0>, rot=<0 0 0 1>}
 weapon_l {parent=hand_l, loc=<-0.1974 -3.4111 0.3811>, rot=<0 0 0 1>}
 hand_attach_l {parent=hand_l, loc=<8.5 -3 0>, rot=<0.7071 0.7071 0 0>}
 lowerarm_twist_02_l {parent=lowerarm_l, loc=<9.6199 0 0>, rot=<0 0 0 -1>}
 lowerarm_twist_01_l {parent=lowerarm_l, loc=<17.8549 0 0>, rot=<0 0 0 -1>}
 deform_elbow_in_base_l {parent=lowerarm_l, loc=<0.2112 5.259 0.8274>, rot=<0 0 0 1>}
 deform_elbow_in_l {parent=deform_elbow_in_base_l, loc=<0 0 0>, rot=<0 0 0 1>}
 deform_elbow_out_base_l {parent=lowerarm_l, loc=<0 -4 0>, rot=<0 0 0 1>}
 deform_elbow_out_l {parent=deform_elbow_out_base_l, loc=<0 0 0>, rot=<0 0 0 1>}
 upperarm_twist_01_l {parent=upperarm_l, loc=<0.0041 0 0>, rot=<0 0 0 -1>}
 deform_upperarm_bk_base_l {parent=upperarm_twist_01_l, loc=<1.0813 -7.2193 -1.9453>, rot=<0 0 0 -1>}
 deform_upperarm_bk_l {parent=deform_upperarm_bk_base_l, loc=<0 0 0>, rot=<0 0 0 1>}
 deform_upperarm_fr_base_l {parent=upperarm_twist_01_l, loc=<-0.3691 7.0506 -2.765>, rot=<0 0 0 -1>}
 deform_upperarm_fr_l {parent=deform_upperarm_fr_base_l, loc=<0 0 0>, rot=<0 0 0 1>}
 deform_upperarm_up_base_l {parent=upperarm_twist_01_l, loc=<-3.2216 -1.1391 3.3311>, rot=<0 0 0 1>}
 deform_upperarm_up_l {parent=deform_upperarm_up_base_l, loc=<0 0 0>, rot=<0 0 0 1>}
 upperarm_l_PullJoint_base {parent=upperarm_twist_01_l, loc=<0.8409 2.6341 -0.0352>, rot=<0 0 0 -1>}
 upperarm_l_PullJoint {parent=upperarm_l_PullJoint_base, loc=<-6.8903 0.0864 -5.1884>, rot=<0 0 0 1>}
 upperarm_twist_02_l {parent=upperarm_l, loc=<13.7941 0 0>, rot=<0 0 0 -1>}
 base_shoulderPad_01_l {parent=clavicle_l, loc=<7.342 -1.4045 9.2484>, rot=<0.703 -0.0862 -0.121 -0.6955>}
 dyn_shoulderPad_01_l {parent=base_shoulderPad_01_l, loc=<0 0 0>, rot=<0 0 0 1>}
 dyn_shoulderPad_02_l {parent=dyn_shoulderPad_01_l, loc=<6.8267 0 0>, rot=<0 0 0 1>}
 base_tasslesStart_l {parent=dyn_shoulderPad_02_l, loc=<6.6535 0 0>, rot=<0 0 0 1>}
 dyn_tasslesStart_l {parent=base_tasslesStart_l, loc=<0 0 0>, rot=<0 0 0 1>}
 base_tassles_C_01_l {parent=dyn_tasslesStart_l, loc=<0.2223 0.9117 -3.2219>, rot=<-0.0934 -0.1002 -0.2068 -0.9688>}
 dyn_tassles_C_01_l {parent=base_tassles_C_01_l, loc=<0 0 0>, rot=<0 0 0 1>}
 dyn_tassles_C_02_l {parent=dyn_tassles_C_01_l, loc=<4.2766 0 0>, rot=<0 0 0 1>}
 base_tassles_B_01_l {parent=dyn_tasslesStart_l, loc=<0.3951 0.5059 0>, rot=<0 0 -0.1999 -0.9798>}
 dyn_tassles_B_01_l {parent=base_tassles_B_01_l, loc=<0 0 0>, rot=<0 0 0 1>}
 dyn_tassles_B_02_l {parent=dyn_tassles_B_01_l, loc=<4.2766 0 0>, rot=<0 0 0 1>}
 base_tassles_A_01_l {parent=dyn_tasslesStart_l, loc=<0.2923 0.7473 3.2895>, rot=<0.1009 0.0661 -0.2034 -0.9716>}
 dyn_tassles_A_01_l {parent=base_tassles_A_01_l, loc=<0 0 0>, rot=<0 0 0 1>}
 dyn_tassles_A_02_l {parent=dyn_tassles_A_01_l, loc=<4.2766 0 0>, rot=<0 0 0 1>}
 base_outChain_01 {parent=dyn_shoulderPad_02_l, loc=<6.2654 1.5889 6.2277>, rot=<0.2668 0.6337 -0.3106 -0.6563>}
 dyn_outChain_01 {parent=base_outChain_01, loc=<0 0 0>, rot=<0 0 0 1>}
 dyn_outChain_02 {parent=dyn_outChain_01, loc=<4.7869 0 0>, rot=<-0.1065 0.3729 -0.0792 -0.9183>}
 dyn_outChain_03 {parent=dyn_outChain_02, loc=<4.6699 0.0601 0.0672>, rot=<0 0 0 1>}
 dyn_outChain_01_PosRef_null {parent=base_outChain_01, type=Null, loc=<0 0 0>, rot=<0 0 0 -1>}
 base_inChain_01 {parent=dyn_shoulderPad_01_l, loc=<1.0577 2.3574 6.9405>, rot=<0.3749 0.4506 -0.2088 -0.7828>}
 dyn_inChain_01 {parent=base_inChain_01, loc=<0 0 0>, rot=<-0.1305 0 0 -0.9914>}
 dyn_inChain_02 {parent=dyn_inChain_01, loc=<4.0833 0 0>, rot=<0.0591 -0.2217 0.0843 -0.9697>}
 dyn_inChain_03 {parent=dyn_inChain_02, loc=<4.4893 -0.0411 0.0017>, rot=<0.0384 -0.1275 -0.0714 -0.9885>}
 dyn_inChain_04 {parent=dyn_inChain_03, loc=<3.4899 0 0>, rot=<0 0 0 1>}
 dyn_inChain_01_PosRef_null {parent=base_inChain_01, type=Null, loc=<0 0 0>, rot=<-0.1305 0 0 -0.9914>}
 base_shoulderPad_aim_l {parent=clavicle_l, loc=<20.3124 -0.7706 3.6842>, rot=<0.703 -0.0862 -0.121 -0.6955>}
 clavicle_r {parent=spine_05, loc=<4.3575 -0.1466 2.3223>, rot=<0.7289 0.0465 -0.6829 0.009>}
 upperarm_r {parent=clavicle_r, loc=<-15.5111 0 0>, rot=<0.0226 -0.3234 -0.0088 -0.9459>}
 lowerarm_r {parent=upperarm_r, loc=<-26.6514 0 0>, rot=<0.0391 -0.0099 -0.2453 -0.9686>}
 hand_r {parent=lowerarm_r, loc=<-22.3491 0 0>, rot=<-0.7081 -0.0507 0.0975 -0.6975>}
 thumb_01_r {parent=hand_r, loc=<-1.3118 0.3191 2.459>, rot=<0.6546 -0.4042 0.0664 -0.6354>}
 thumb_02_r {parent=thumb_01_r, loc=<-5.9336 0.0448 -0.0401>, rot=<-0.0037 0.0213 0.298 -0.9543>}
 thumb_03_r {parent=thumb_02_r, loc=<-2.2223 -0.3682 0.0907>, rot=<0.0118 -0.013 0.0066 -0.9998>}
 index_metacarpal_r {parent=hand_r, loc=<-1.6535 -0.0183 2.96>, rot=<0.0158 -0.0124 -0.0173 -0.9996>}
 index_01_r {parent=index_metacarpal_r, loc=<-8.3008 -0.0357 0.0524>, rot=<-0.0044 0.0485 0.0924 -0.9945>}
 index_02_r {parent=index_01_r, loc=<-3.3587 0.3004 -0.0452>, rot=<-0.002 0.0027 0.3225 -0.9466>}
 index_03_r {parent=index_02_r, loc=<-2.5578 0.123 0.0252>, rot=<-0.002 0.0067 0.2558 -0.9667>}
 middle_metacarpal_r {parent=hand_r, loc=<-2.1856 0.1725 0.5984>, rot=<0.006 0.0168 -0.0329 -0.9993>}
 middle_01_r {parent=middle_metacarpal_r, loc=<-7.3275 0.2447 -0.0129>, rot=<-0.0208 0.0042 0.0858 -0.9961>}
 middle_02_r {parent=middle_01_r, loc=<-4.0123 -0.1401 -0.1175>, rot=<-0.0218 0.0258 0.3512 -0.9357>}
 middle_03_r {parent=middle_02_r, loc=<-2.6077 -0.0534 -0.1373>, rot=<0.0082 0.0188 0.1708 -0.9851>}
 ring_metacarpal_r {parent=hand_r, loc=<-2.2074 0.3625 -1.3638>, rot=<-0.0284 0.0722 -0.0458 -0.9959>}
 ring_01_r {parent=ring_metacarpal_r, loc=<-7.0295 0.6197 0.1446>, rot=<-0.0489 -0.0252 0.1892 -0.9804>}
 ring_02_r {parent=ring_01_r, loc=<-3.6174 -0.4647 0.1027>, rot=<0 0 0.2806 -0.9598>}
 ring_03_r {parent=ring_02_r, loc=<-2.4152 -0.114 -0.0199>, rot=<0 0 0.1189 -0.9929>}
 pinky_metacarpal_r {parent=hand_r, loc=<-1.7616 0.9753 -3.3814>, rot=<-0.0397 0.0814 -0.069 -0.9935>}
 pinky_01_r {parent=pinky_metacarpal_r, loc=<-7.1709 0.4831 0.1094>, rot=<-0.0654 -0.0035 0.2242 -0.9723>}
 pinky_02_r {parent=pinky_01_r, loc=<-2.8274 -0.0893 0.1043>, rot=<0.0037 0.0011 0.2854 -0.9584>}
 pinky_03_r {parent=pinky_02_r, loc=<-2.2046 -0.0307 0.0078>, rot=<-0.0002 0.0098 0.1445 -0.9895>}
 deform_wrist_dn_base_r {parent=hand_r, loc=<-0.0004 2.2345 -0.0001>, rot=<0 0 0 -1>}
 deform_wrist_dn_r {parent=deform_wrist_dn_base_r, loc=<0 0 0>, rot=<0 0 0 1>}
 deform_wrist_up_base_r {parent=hand_r, loc=<0.0003 -2.2684 0.0001>, rot=<0 0 0 -1>}
 deform_wrist_up_r {parent=deform_wrist_up_base_r, loc=<0 0 0>, rot=<0 0 0 1>}
 weapon_r {parent=hand_r, loc=<0.1974 3.4111 -0.3811>, rot=<0 0 0 1>}
 hand_attach_r {parent=hand_r, loc=<-8.5001 2.9999 0>, rot=<0 0 -0.7071 -0.7071>}
 lowerarm_twist_02_r {parent=lowerarm_r, loc=<-9.6199 0 0>, rot=<0 0 0 -1>}
 lowerarm_twist_01_r {parent=lowerarm_r, loc=<-17.8549 0 0>, rot=<0 0 0 -1>}
 deform_elbow_in_base_r {parent=lowerarm_r, loc=<-0.2111 -5.2591 -0.8275>, rot=<0 0 0 1>}
 deform_elbow_in_r {parent=deform_elbow_in_base_r, loc=<-0.0001 0 0.0002>, rot=<0 0 0 1>}
 deform_elbow_out_base_r {parent=lowerarm_r, loc=<0 4 -0.0001>, rot=<0 0 0 1>}
 deform_elbow_out_r {parent=deform_elbow_out_base_r, loc=<0 0 0>, rot=<0 0 0 1>}
 upperarm_twist_01_r {parent=upperarm_r, loc=<-0.0041 0 0>, rot=<0 0 0 -1>}
 deform_upperarm_bk_base_r {parent=upperarm_twist_01_r, loc=<-1.0814 7.2192 1.9455>, rot=<0 0 0 1>}
 deform_upperarm_bk_r {parent=deform_upperarm_bk_base_r, loc=<0 0 0>, rot=<0 0 0 1>}
 deform_upperarm_fr_base_r {parent=upperarm_twist_01_r, loc=<0.3689 -7.0506 2.7652>, rot=<0 0 0 1>}
 deform_upperarm_fr_r {parent=deform_upperarm_fr_base_r, loc=<0 0 0>, rot=<0 0 0 1>}
 deform_upperarm_up_base_r {parent=upperarm_twist_01_r, loc=<3.2214 1.1391 -3.3308>, rot=<0 0 0 1>}
 deform_upperarm_up_r {parent=deform_upperarm_up_base_r, loc=<0 0 0>, rot=<0 0 0 -1>}
 upperarm_r_PullJoint_base {parent=upperarm_twist_01_r, loc=<-0.8408 -2.6341 0.0351>, rot=<0 0 0 -1>}
 upperarm_r_PullJoint {parent=upperarm_r_PullJoint_base, loc=<8.8514 -1.1868 7.6811>, rot=<0 0 0 1>}
 upperarm_twist_02_r {parent=upperarm_r, loc=<-13.7941 0 0>, rot=<0 0 0 -1>}
 base_shoulderPad_01_r {parent=clavicle_r, loc=<-7.3419 1.4045 -9.2485>, rot=<-0.0862 -0.703 -0.6955 0.121>}
 dyn_shoulderPad_01_r {parent=base_shoulderPad_01_r, loc=<0 0 0>, rot=<0 0 0 -1>}
 dyn_shoulderPad_02_r {parent=dyn_shoulderPad_01_r, loc=<6.8268 0 0>, rot=<0 0 0 -1>}
 base_tasslesStart_r {parent=dyn_shoulderPad_02_r, loc=<6.6532 0 0>, rot=<0 0 0 1>}
 dyn_tasslesStart_r {parent=base_tasslesStart_r, loc=<0 0 0>, rot=<0 0 0 1>}
 base_tassles_A_01_r {parent=dyn_tasslesStart_r, loc=<0.2225 0.9117 -3.2219>, rot=<-0.098 -0.0788 -0.2046 -0.9707>}
 dyn_tassles_A_01_r {parent=base_tassles_A_01_r, loc=<0 0 0>, rot=<0 0 0 1>}
 dyn_tassles_A_02_r {parent=dyn_tassles_A_01_r, loc=<4.2766 0 0>, rot=<0 0 0 1>}
 base_tassles_B_01_r {parent=dyn_tasslesStart_r, loc=<0.3953 0.506 0>, rot=<0 0 -0.1999 -0.9798>}
 dyn_tassles_B_01_r {parent=base_tassles_B_01_r, loc=<0 0 0>, rot=<0 0 0 1>}
 dyn_tassles_B_02_r {parent=dyn_tassles_B_01_r, loc=<4.2766 0 0>, rot=<0 0 0 1>}
 base_tassles_C_01_r {parent=dyn_tasslesStart_r, loc=<0.2925 0.7474 3.2895>, rot=<0.0962 0.0883 -0.2056 -0.9699>}
 dyn_tassles_C_01_r {parent=base_tassles_C_01_r, loc=<0 0 0>, rot=<0 0 0 1>}
 dyn_tassles_C_02_r {parent=dyn_tassles_C_01_r, loc=<4.2766 0 0>, rot=<0 0 0 1>}
 base_shoulderPad_aim_r {parent=clavicle_r, loc=<-20.3123 0.7706 -3.6843>, rot=<-0.0862 -0.703 -0.6955 0.121>}
 base_jacktLapel_l {parent=spine_05, loc=<4.9092 9.8451 -7.1205>, rot=<0.7011 -0.2075 -0.6621 -0.1645>}
 dyn_jacktLapel_l {parent=base_jacktLapel_l, loc=<0 0 0>, rot=<0 0 0 1>}
 dyn_jacktLapel_end_l {parent=dyn_jacktLapel_l, loc=<8.6952 0 0>, rot=<0 0 0 1>}
 base_jacktLapel_r {parent=spine_05, loc=<4.9094 9.8452 7.1205>, rot=<0.7011 -0.2075 0.6621 0.1645>}
 dyn_jacktLapel_r {parent=base_jacktLapel_r, loc=<0 0 0>, rot=<0 0 0 1>}
 dyn_jacktLapel_end_r {parent=dyn_jacktLapel_r, loc=<8.6953 0 0>, rot=<0 0 0 1>}
 base_neckLapel_01_l {parent=spine_05, loc=<10.9542 4.9839 -6.5924>, rot=<0.655 -0.1438 -0.6486 -0.3599>}
 dyn_neckLapel_01_l {parent=base_neckLapel_01_l, loc=<0 0 0>, rot=<0 0 0 1>}
 dyn_jacktLapel_l_end {parent=dyn_neckLapel_01_l, loc=<4.8578 0 0>, rot=<0 0 0 1>}
 base_neckLapel_01_r {parent=spine_05, loc=<10.9568 4.9997 6.5873>, rot=<0.6331 -0.1671 0.6419 0.3992>}
 dyn_neckLapel_01_r {parent=base_neckLapel_01_r, loc=<0 0 0>, rot=<0 0 0 1>}
 dyn_jacktLapel_r_end {parent=dyn_neckLapel_01_r, loc=<4.8578 0 0>, rot=<0 0 0 1>}
 base_neckLapel_02_l {parent=spine_05, loc=<14.6059 -0.2105 -6.6109>, rot=<0.8125 -0.1148 -0.3517 -0.4506>}
 dyn_neckLapel_02_l {parent=base_neckLapel_02_l, loc=<0 0 0>, rot=<0 0 0 1>}
 dyn_jacktLapel_02_l_end {parent=dyn_neckLapel_02_l, loc=<4.8578 0 0>, rot=<0 0 0 1>}
 base_neckLapel_b {parent=spine_05, loc=<16.2378 -3.3679 0.0063>, rot=<1 -0.0082 0 0>}
 dyn_neckLapel_b {parent=base_neckLapel_b, loc=<0 0 0>, rot=<0 0 0 1>}
 dyn_jacktLapel_b_end {parent=dyn_neckLapel_b, loc=<5.5414 0 0>, rot=<0 0 0 1>}
 base_neckLapel_02_r {parent=spine_05, loc=<14.6253 -0.2453 6.5974>, rot=<0.7963 -0.1203 0.3559 0.474>}
 dyn_neckLapel_02_r {parent=base_neckLapel_02_r, loc=<0 0 0>, rot=<0 0 0 1>}
 dyn_jacktLapel_02_r_end {parent=dyn_neckLapel_02_r, loc=<4.8578 0 0>, rot=<0 0 0 1>}
 base_necklace_01 {parent=spine_05, loc=<10.6988 4.9635 0>, rot=<0 0 0.9286 0.3711>}
 dyn_necklace_01 {parent=base_necklace_01, loc=<0 0 0>, rot=<0 0 0 1>}
 dyn_necklace_02 {parent=dyn_necklace_01, loc=<3.055 0 0>, rot=<0 0 -0.1322 -0.9912>}
 base_dogTag_01 {parent=dyn_necklace_02, loc=<2.465 0.2369 0>, rot=<0 0 0 1>}
 dyn_dogTag_01 {parent=base_dogTag_01, loc=<0 0 0>, rot=<0 0 0 1>}
 dyn_dogTag_end {parent=dyn_dogTag_01, loc=<4.6374 0 0>, rot=<0 0 0 1>}
 base_medal_01 {parent=spine_05, loc=<1.0985 13.1598 11.0759>, rot=<0.0249 0.0062 0.9866 0.161>}
 dyn_medal_01 {parent=base_medal_01, loc=<0 0 0>, rot=<0 0 0 1>}
 dyn_medalStar_02 {parent=dyn_medal_01, loc=<2.9902 -0.0008 -0.0662>, rot=<0.0066 -0.0274 -0.0845 -0.996>}
 dyn_medalPendant_03 {parent=dyn_medalStar_02, loc=<4.4998 0.2923 -0.0118>, rot=<0.0032 -0.0102 -0.0648 -0.9978>}
 dyn_medalString_04 {parent=dyn_medalPendant_03, loc=<2.1706 0 0>, rot=<0 0 0 1>}
 base_neckTweak {parent=spine_05, loc=<14.4635 -5.7228 0.001>, rot=<0.9722 0.2343 0 0>}
 dyn_neckTweak {parent=base_neckTweak, loc=<0 0 0>, rot=<0 0 0 1>}
 deform_pec_aim_base_l {parent=spine_04, loc=<7.8732 12.6388 -2.4123>, rot=<0.0102 -0.4602 0.3979 -0.7936>}
 deform_pec_aim_l {parent=deform_pec_aim_base_l, loc=<0 0 0>, rot=<0 0 0 1>}
 deform_pec_base_l {parent=deform_pec_aim_l, loc=<10.766 0 0>, rot=<0 0 0 1>}
 deform_pec_l {parent=deform_pec_base_l, loc=<0.055 0 0>, rot=<0 0 0 1>}
 deform_pec_aim_base_r {parent=spine_04, loc=<7.8727 12.6388 2.4123>, rot=<-0.4602 -0.0102 0.7936 0.3979>}
 deform_pec_aim_r {parent=deform_pec_aim_base_r, loc=<0 0 0>, rot=<0 0 0 1>}
 deform_pec_base_r {parent=deform_pec_aim_r, loc=<-10.7662 -0.0006 -0.0007>, rot=<0 0 0 1>}
 deform_pec_r {parent=deform_pec_base_r, loc=<0 0 0>, rot=<0 0 0 1>}
 base_jacktLapel_low_l {parent=spine_04, loc=<7.3122 12.8518 -7.2604>, rot=<0.7064 -0.1362 -0.6874 -0.0997>}
 dyn_jacktLapel_low_l {parent=base_jacktLapel_low_l, loc=<0 0 0>, rot=<0 0 0 1>}
 dyn_jacktLapel_low_end_l {parent=dyn_jacktLapel_low_l, loc=<6.0543 0 0>, rot=<0 0 0 1>}
 base_jacktLapel_row_r {parent=spine_04, loc=<7.2919 12.9138 6.1425>, rot=<0.7081 -0.124 0.6897 0.0871>}
 dyn_jacktLapel_row_r {parent=base_jacktLapel_row_r, loc=<0 0 0>, rot=<0 0 0 1>}
 base_jacktLapel_low_r {parent=spine_04, loc=<7.3119 12.8518 7.2604>, rot=<0.7081 -0.124 0.6897 0.0871>}
 dyn_jacktLapel_low_r {parent=base_jacktLapel_low_r, loc=<0 0 0>, rot=<0 0 0 1>}
 dyn_jacktLapel_low_end_r {parent=dyn_jacktLapel_low_r, loc=<6.0543 -0.0001 0.0003>, rot=<0 0 0 1>}
 base_pouch_01 {parent=spine_01, loc=<0.8466 11.6093 2.2007>, rot=<-0.002 -0.0383 0.9971 -0.0657>}
 dyn_pouch_01 {parent=base_pouch_01, loc=<0 0 0>, rot=<0 0 0 1>}
 dyn_pouch_01_end {parent=dyn_pouch_01, loc=<6.1448 0 0>, rot=<0 0 0 1>}
 base_pouch_02 {parent=spine_01, loc=<1.0542 9.4784 8.6466>, rot=<0.0067 -0.2571 0.9648 -0.0555>}
 dyn_pouch_02 {parent=base_pouch_02, loc=<0 0 0>, rot=<0 0 0 1>}
 dyn_pouch_02_end {parent=dyn_pouch_02, loc=<6.1449 0 0>, rot=<0 0 0 1>}
 base_frontChain_01 {parent=spine_01, loc=<1.0946 10.6978 -5.6231>, rot=<-0.04 0.0846 0.9956 -0.0028>}
 dyn_frontChain_01 {parent=base_frontChain_01, loc=<0 0 0>, rot=<0 0 0 1>}
 dyn_frontChain_02 {parent=dyn_frontChain_01, loc=<3.6093 0 0>, rot=<-0.0276 -0.132 0.0145 -0.9908>}
 dyn_frontChain_03 {parent=dyn_frontChain_02, loc=<4.974 -0.0401 -0.009>, rot=<-0.0507 -0.2645 -0.0915 -0.9587>}
 dyn_frontChain_04 {parent=dyn_frontChain_03, loc=<3.4899 0 0>, rot=<0 0 0 1>}
 dyn_frontChain_01_PosRef_null {parent=dyn_frontChain_01, type=Null, loc=<0 0 0>, rot=<0 0 0 1>}
 base_backChain_01 {parent=spine_01, loc=<-1.4612 6.2854 -12.454>, rot=<-0.0323 0.0878 0.9914 -0.0915>}
 dyn_backChain_01 {parent=base_backChain_01, loc=<0 0 0>, rot=<0 0 0.0692 -0.9976>}
 dyn_backChain_02 {parent=dyn_backChain_01, loc=<4.9002 -1.1427 0.6289>, rot=<0.0392 0.3278 0.2459 -0.9113>}
 dyn_backChain_03 {parent=dyn_backChain_02, loc=<4.1801 0 0>, rot=<0 0 0 1>}
 dyn_BackChain_01_PosRef_null {parent=dyn_backChain_01, type=Null, loc=<0 0 0>, rot=<0 0 0 1>}
 base_jacketFlap_Front_01_l {parent=spine_01, loc=<3.8718 3.1538 -13.5329>, rot=<-0.096 0.694 0.7134 0.0135>}
 dyn_jacketFlap_Front_01_l {parent=base_jacketFlap_Front_01_l, loc=<0 0 0>, rot=<0 0 0 1>}
 dyn_jacketFlap_Front_02_l {parent=dyn_jacketFlap_Front_01_l, loc=<6.8248 0 0>, rot=<0 0 0 1>}
 base_jacketFlap_Side_01_l {parent=spine_01, loc=<3.7572 -5.6376 -12.1568>, rot=<0.1446 -0.8502 -0.5062 0.0006>}
 dyn_jacketFlap_Side_01_l {parent=base_jacketFlap_Side_01_l, loc=<0 0 0>, rot=<0 0 0 1>}
 dyn_jacketFlap_Side_02_l {parent=dyn_jacketFlap_Side_01_l, loc=<6.4273 0 0>, rot=<0 0 0 1>}
 base_jacketFlap_Back_01_l {parent=spine_01, loc=<3.4983 -10.9746 -4.2708>, rot=<0.0982 -0.9874 -0.1235 -0.0089>}
 dyn_jacketFlap_Back_01_l {parent=base_jacketFlap_Back_01_l, loc=<0 0 0>, rot=<0 0 0 1>}
 dyn_jacketFlap_Back_02_l {parent=dyn_jacketFlap_Back_01_l, loc=<6.4273 0 0>, rot=<0 0 0 1>}
 base_jacketFlap_Back_01_r {parent=spine_01, loc=<3.4983 -10.9746 4.271>, rot=<0.0982 -0.9874 0.1235 0.0089>}
 dyn_jacketFlap_Back_01_r {parent=base_jacketFlap_Back_01_r, loc=<0 0 0>, rot=<0 0 0 1>}
 dyn_jacketFlap_Back_02_r {parent=dyn_jacketFlap_Back_01_r, loc=<6.4273 0 0>, rot=<0 0 0 1>}
 base_jacketFlap_Side_01_r {parent=spine_01, loc=<3.7572 -5.6376 12.1572>, rot=<0.1446 -0.8502 0.5062 -0.0006>}
 dyn_jacketFlap_Side_01_r {parent=base_jacketFlap_Side_01_r, loc=<0 0 0>, rot=<0 0 0 1>}
 dyn_jacketFlap_Side_02_r {parent=dyn_jacketFlap_Side_01_r, loc=<6.4273 0 0>, rot=<0 0 0 1>}
 base_jacketFlap_Front_01_r {parent=spine_01, loc=<3.8723 3.1538 13.5329>, rot=<0.096 -0.694 0.7134 0.0135>}
 dyn_jacketFlap_Front_01_r {parent=base_jacketFlap_Front_01_r, loc=<0 0 0>, rot=<0 0 0 1>}
 dyn_jacketFlap_Front_02_r {parent=dyn_jacketFlap_Front_01_r, loc=<6.8252 0.0001 0>, rot=<0 0 0 1>}
 thigh_l {parent=pelvis, loc=<-2.5693 -0.0109 -7.7501>, rot=<0.0012 0.0572 0.0155 -0.9982>}
 calf_l {parent=thigh_l, loc=<-39.0484 0 0>, rot=<0.0006 -0.0161 -0.0391 -0.9991>}
 foot_l {parent=calf_l, loc=<-42.561 0.1077 0.4292>, rot=<0.0418 -0.0455 0.0343 -0.9975>}
 ball_l {parent=foot_l, loc=<-6.1599 13.8264 0.6806>, rot=<0.0001 -0.0001 -0.7063 -0.7079>}
 calf_twist_02_l {parent=calf_l, loc=<-18.8064 -0.0967 0.0397>, rot=<0.0237 0.0095 -0.0035 -0.9997>}
 calf_twist_01_l {parent=calf_l, loc=<-35.813 -0.0967 0.0397>, rot=<0.0237 0.0095 -0.0035 -0.9997>}
 deform_knee_in_base_l {parent=calf_l, loc=<1.298 -5.2957 -0.0496>, rot=<0 0 0 -1>}
 deform_knee_in_l {parent=deform_knee_in_base_l, loc=<0 0 0>, rot=<0 0 0 1>}
 deform_knee_out_base_l {parent=calf_l, loc=<0 4 0>, rot=<0 0 0 -1>}
 deform_knee_out_l {parent=deform_knee_out_base_l, loc=<0 0 0>, rot=<0 0 0 1>}
 thigh_twist_01_l {parent=thigh_l, loc=<-21.6303 0 0>, rot=<0.0013 0 0 -1>}
 deform_groin_base_l {parent=thigh_l, loc=<-2.1908 8.1142 -1.7442>, rot=<0 0 0 -1>}
 deform_groin_l {parent=deform_groin_base_l, loc=<0 0 0>, rot=<0 0 0 1>}
 base_legPouch_01 {parent=thigh_l, loc=<-8.6466 0.8013 -8.8547>, rot=<0.0334 0.6684 0.7428 -0.0184>}
 dyn_legPouch_01 {parent=base_legPouch_01, loc=<0 0 0>, rot=<0 0 0 1>}
 dyn_legPouch_01_end {parent=dyn_legPouch_01, loc=<8.1644 0 0>, rot=<0 0 0 1>}
 thigh_r {parent=pelvis, loc=<-2.5693 -0.0109 7.7501>, rot=<0.0572 -0.0012 0.9982 0.0155>}
 calf_r {parent=thigh_r, loc=<39.0484 0 0>, rot=<0.0006 -0.0161 -0.0391 -0.9991>}
 foot_r {parent=calf_r, loc=<42.561 -0.1077 -0.4292>, rot=<0.0418 -0.0455 0.0343 -0.9975>}
 ball_r {parent=foot_r, loc=<6.1599 -13.8264 -0.6806>, rot=<0.0001 -0.0001 -0.7063 -0.7079>}
 calf_twist_02_r {parent=calf_r, loc=<18.8064 0.0967 -0.0397>, rot=<0.0237 0.0095 -0.0035 -0.9997>}
 calf_twist_01_r {parent=calf_r, loc=<35.813 0.0967 -0.0397>, rot=<0.0237 0.0095 -0.0035 -0.9997>}
 deform_knee_in_base_r {parent=calf_r, loc=<-1.298 5.2957 0.0496>, rot=<0 0 0 -1>}
 deform_knee_in_r {parent=deform_knee_in_base_r, loc=<0 0 0>, rot=<0 0 0 1>}
 deform_knee_out_base_r {parent=calf_r, loc=<0 -4 0>, rot=<0 0 0 -1>}
 deform_knee_out_r {parent=deform_knee_out_base_r, loc=<0 0 0>, rot=<0 0 0 1>}
 base_kneeCover {parent=calf_r, loc=<0.5119 -3.7855 0.8916>, rot=<-0.0018 -0.0216 -0.0032 -0.9998>}
 dyn_kneeCover {parent=base_kneeCover, loc=<0 0 0>, rot=<0 0 0 1>}
 thigh_twist_01_r {parent=thigh_r, loc=<21.6303 0 0>, rot=<0.0013 0 0 -1>}
 deform_groin_base_r {parent=thigh_r, loc=<2.1908 -8.1142 1.7442>, rot=<0 0 0 -1>}
 deform_groin_r {parent=deform_groin_base_r, loc=<0 0 0>, rot=<0 0 0 1>}
 deform_glute_base_l {parent=pelvis, loc=<-8.8064 -8.8222 -8.4287>, rot=<0.0275 0.0501 -0.4472 -0.8926>}
 deform_glute_l {parent=deform_glute_base_l, loc=<0 0 0>, rot=<0 0 0 1>}
 deform_glute_base_r {parent=pelvis, loc=<-8.8062 -8.8222 8.4287>, rot=<0.0501 -0.0275 0.8926 -0.4472>}
 deform_glute_r {parent=deform_glute_base_r, loc=<0 0 0>, rot=<0 0 0 1>}
 attach {parent=root, loc=<0 0 0>, rot=<0 0 0 1>}
 ik_hand_root {parent=root, loc=<0 0 0>, rot=<0 0 0 1>}
 ik_hand_gun {parent=ik_hand_root, loc=<-51.9727 4.0883 102.6935>, rot=<0.6084 0.1793 0.4613 -0.6205>}
 ik_hand_r {parent=ik_hand_gun, loc=<0 0 0>, rot=<0 0 0 1>}
 ik_hand_l {parent=ik_hand_gun, loc=<53.0285 82.1781 35.2051>, rot=<-0.5102 -0.7906 -0.3387 0>}
 ik_foot_root {parent=root, loc=<0 0 0>, rot=<0 0 0 1>}
 ik_foot_l {parent=ik_foot_root, loc=<15.2608 -4.4176 9.4432>, rot=<0.0357 0.7045 0.0284 -0.7082>}
 ik_foot_r {parent=ik_foot_root, loc=<-15.2608 -4.4176 9.4432>, rot=<0.7082 0.0284 -0.7045 0.0357>}
]
variables [Pub_ISCelebreteEmote {type=double, default=0}]
