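# Physics asset for the TacticalPrince male medium cosmetic body's tassels, giving left and right tassel bone chains collision capsules and joints.
# Each clavicle carries a start bone pair that branches into three tassel strands (A, B, C), each a base bone followed by two dyn links.
# Dyn joints have angular limits, with the start joint bending on one swing axis, while base and body joints are free.
# Larger spine, clavicle and upper-arm capsules form the body volume.
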PhysicsAsset /BRCosmetics/Characters/Player/Male/Medium/Bodies/M_MED_TacticalPrince/Meshes/M_MED_TacticalPrince_Tassles_Physics
bodies {
 base_tasslesStart_l [{
   shapeType=Capsule
   length=1.1809
   radius=0.5483
   shapeName=base_tasslesStart_l_capsule
   center=<-0.1803 -0.4329 0.5453>
   rotation=<rot 0 -90 0>
  }]
 base_tasslesStart_r [{shapeType=Capsule, length=0.505, radius=0.505, shapeName=base_tasslesStart_r_capsule, rotation=<rot 0 -90 0>}]
 base_tassles_A_01_l [{shapeType=Capsule, length=0.505, radius=0.505, shapeName=base_tassles_A_01_l_capsule, rotation=<rot 0 -90 0>}]
 base_tassles_A_01_r [{shapeType=Capsule, length=0.505, radius=0.505, shapeName=base_tassles_A_01_r_capsule, rotation=<rot 0 -90 0>}]
 base_tassles_B_01_l [{shapeType=Capsule, length=0.505, radius=0.505, shapeName=base_tassles_B_01_l_capsule, rotation=<rot 0 -90 0>}]
 base_tassles_B_01_r [{shapeType=Capsule, length=0.505, radius=0.505, shapeName=base_tassles_B_01_r_capsule, rotation=<rot 0 -90 0>}]
 base_tassles_C_01_l [{shapeType=Capsule, length=0.505, radius=0.505, shapeName=base_tassles_C_01_l_capsule, rotation=<rot 0 -90 0>}]
 base_tassles_C_01_r [{shapeType=Capsule, length=0.505, radius=0.505, shapeName=base_tassles_C_01_r_capsule, rotation=<rot 0 -90 0>}]
 clavicle_l [{
   shapeType=Capsule
   length=6.8794
   radius=7.997
   shapeName=clavicle_l_capsule
   center=<5.9142 -1.2691 1.781>
   rotation=<rot 71.4369 -0.0002 -0.0002>
  }]
 clavicle_r [{
   shapeType=Capsule
   length=6.8794
   radius=7.997
   shapeName=clavicle_r_capsule
   center=<-5.9142 1.2691 -1.781>
   rotation=<rot 71.4369 -0.0002 179.9998>
  }]
 dyn_tasslesStart_l [{
   shapeType=Capsule
   length=6.0911
   radius=1.6002
   shapeName=dyn_tasslesStart_l_capsule
   center=<2.6096 1.2068 -0.2446>
   rotation=<rot -4.9279 -89.9826 1.7646>
  }]
 dyn_tasslesStart_r [{
   shapeType=Capsule
   length=6.0911
   radius=1.6002
   shapeName=dyn_tasslesStart_r_capsule
   center=<3.7523 1.953 0.3053>
   rotation=<rot -0.2333 -69.908 179.9686>
  }]
 dyn_tassles_A_01_l [{
   shapeType=Capsule
   length=2.8544
   radius=1.0662
   shapeName=dyn_tassles_A_01_l_capsule
   center=<2.2202 0.4714 0.0122>
   rotation=<rot 11.118 -90.2136 178.7824>
  }]
 dyn_tassles_A_01_r [{
   shapeType=Capsule
   length=2.8544
   radius=1.0662
   shapeName=dyn_tassles_A_01_r_capsule
   center=<2.425 0.3668 -0.0667>
   rotation=<rot -13.8758 -89.684 -1.23>
  }]
 dyn_tassles_A_02_l [{
   shapeType=Capsule
   length=4.013
   radius=1.642
   shapeName=dyn_tassles_C_01_l_capsule
   center=<3.425 0.3303 0.681>
   rotation=<rot 14.4277 -86.5204 178.5453>
  }]
 dyn_tassles_A_02_r [{
   shapeType=Capsule
   length=3.4481
   radius=1.8291
   shapeName=dyn_tassles_C_01_r_capsule
   center=<3.8354 0.2637 -0.9543>
   rotation=<rot -15.5624 -85.7659 -1.4617>
  }]
 dyn_tassles_B_01_l [{
   shapeType=Capsule
   length=3.3991
   radius=1.2093
   shapeName=dyn_tassles_B_01_l_capsule
   center=<2.1212 0.2411 -0.3826>
   rotation=<rot -3.8847 -89.9996 179.2534>
  }]
 dyn_tassles_B_01_r [{
   shapeType=Capsule
   length=3.3991
   radius=1.2093
   shapeName=dyn_tassles_B_01_r_capsule
   center=<2.5212 0.033 0.5037>
   rotation=<rot 1.1154 -90.0646 -0.7445>
  }]
 dyn_tassles_B_02_l [{
   shapeType=Capsule
   length=4.9364
   radius=1.642
   shapeName=dyn_tassles_C_01_l_capsule
   center=<3.4835 -0.2462 -0.1217>
   rotation=<rot -5.5662 -86.0203 178.5845>
  }]
 dyn_tassles_B_02_r [{
   shapeType=Capsule
   length=4.3685
   radius=1.9532
   shapeName=dyn_tassles_C_01_r_capsule
   center=<3.85 -0.3385 0.0956>
   rotation=<rot -0.5672 -86.1437 -1.4083>
  }]
 dyn_tassles_C_01_l [{
   shapeType=Capsule
   length=2.3606
   radius=1.2548
   shapeName=dyn_tassles_C_01_l_capsule
   center=<2.1033 0.5055 -1.1851>
   rotation=<rot -15.5629 -85.765 178.5377>
  }]
 dyn_tassles_C_01_r [{
   shapeType=Capsule
   length=2.3606
   radius=1.2548
   shapeName=dyn_tassles_C_01_r_capsule
   center=<2.4569 0.7858 1.5223>
   rotation=<rot 9.43 -86.3911 -1.4274>
  }]
 dyn_tassles_C_02_l [{
   shapeType=Capsule
   length=3.7482
   radius=1.642
   shapeName=dyn_tassles_C_01_l_capsule
   center=<3.4632 0.4263 -1.1306>
   rotation=<rot -15.5629 -85.765 178.5377>
  }]
 dyn_tassles_C_02_r [{
   shapeType=Capsule
   length=2.9447
   radius=2.1505
   shapeName=dyn_tassles_C_01_r_capsule
   center=<3.8821 0.6151 1.1475>
   rotation=<rot 20.0029 -86.6698 -1.4984>
  }]
 spine_04 [{
   shapeType=Capsule
   length=4.5248
   radius=14.1194
   shapeName=spine_04_capsule
   center=<3.0856 2.0492 -0.0521>
   rotation=<rot -0.3302 -90 0.2757>
  }]
 spine_05 [{shapeType=Capsule, length=11.3313, radius=12.035, shapeName=spine_05_capsule, center=<0.3672 0.5477 0.4414>, rotation=<rot 0 -90 0>}]
 upperarm_l [{
   shapeType=Capsule
   length=14.1227
   radius=7.6416
   shapeName=upperarm_l_capsule
   center=<8.222 -0.6953 -2.6732>
   rotation=<rot 87.2675 -0.0071 -0.0075>
  }]
 upperarm_r [{
   shapeType=Capsule
   length=14.3308
   radius=7.4335
   shapeName=upperarm_r_capsule
   center=<-8.5822 0.4117 2.7557>
   rotation=<rot 87.2952 -3.7753 176.0547>
  }]
}
constraints [
 {
  bone1Name=spine_05
  bone2Name=spine_04
  swing1LimitDegrees=45
  swing1Motion=Free
  swing2LimitDegrees=45
  swing2Motion=Free
  twistLimitDegrees=45
  twistMotion=Free
 }
 {
  bone1Name=clavicle_l
  bone2Name=spine_05
  swing1LimitDegrees=45
  swing1Motion=Free
  swing2LimitDegrees=45
  swing2Motion=Free
  twistLimitDegrees=45
  twistMotion=Free
 }
 {
  bone1Name=upperarm_l
  bone2Name=clavicle_l
  swing1LimitDegrees=45
  swing1Motion=Free
  swing2LimitDegrees=45
  swing2Motion=Free
  twistLimitDegrees=45
  twistMotion=Free
 }
 {
  bone1Name=clavicle_r
  bone2Name=spine_05
  swing1LimitDegrees=45
  swing1Motion=Free
  swing2LimitDegrees=45
  swing2Motion=Free
  twistLimitDegrees=45
  twistMotion=Free
 }
 {
  bone1Name=upperarm_r
  bone2Name=clavicle_r
  swing1LimitDegrees=45
  swing1Motion=Free
  swing2LimitDegrees=45
  swing2Motion=Free
  twistLimitDegrees=45
  twistMotion=Free
 }
 {
  bone1Name=base_tasslesStart_l
  bone2Name=clavicle_l
  swing1LimitDegrees=45
  swing1Motion=Free
  swing2LimitDegrees=45
  swing2Motion=Free
  twistLimitDegrees=45
  twistMotion=Free
 }
 {bone1Name=dyn_tasslesStart_l, bone2Name=base_tasslesStart_l, swing1LimitDegrees=20, swing2LimitDegrees=0, twistLimitDegrees=0}
 {
  bone1Name=base_tassles_A_01_l
  bone2Name=dyn_tasslesStart_l
  swing1LimitDegrees=45
  swing1Motion=Free
  swing2LimitDegrees=45
  swing2Motion=Free
  twistLimitDegrees=45
  twistMotion=Free
 }
 {bone1Name=dyn_tassles_A_01_l, bone2Name=base_tassles_A_01_l, swing1LimitDegrees=50, swing2LimitDegrees=15, twistLimitDegrees=5}
 {bone1Name=dyn_tassles_A_02_l, bone2Name=dyn_tassles_A_01_l, swing1LimitDegrees=30, swing2LimitDegrees=15, twistLimitDegrees=20}
 {
  bone1Name=base_tassles_B_01_l
  bone2Name=dyn_tasslesStart_l
  swing1LimitDegrees=45
  swing1Motion=Free
  swing2LimitDegrees=45
  swing2Motion=Free
  twistLimitDegrees=45
  twistMotion=Free
 }
 {bone1Name=dyn_tassles_B_01_l, bone2Name=base_tassles_B_01_l, swing1LimitDegrees=50, swing2LimitDegrees=15, twistLimitDegrees=5}
 {bone1Name=dyn_tassles_B_02_l, bone2Name=dyn_tassles_B_01_l, swing1LimitDegrees=30, swing2LimitDegrees=15, twistLimitDegrees=20}
 {
  bone1Name=base_tassles_C_01_l
  bone2Name=dyn_tasslesStart_l
  swing1LimitDegrees=45
  swing1Motion=Free
  swing2LimitDegrees=45
  swing2Motion=Free
  twistLimitDegrees=45
  twistMotion=Free
 }
 {bone1Name=dyn_tassles_C_01_l, bone2Name=base_tassles_C_01_l, swing1LimitDegrees=50, swing2LimitDegrees=15, twistLimitDegrees=5}
 {bone1Name=dyn_tassles_C_02_l, bone2Name=dyn_tassles_C_01_l, swing1LimitDegrees=30, swing2LimitDegrees=15, twistLimitDegrees=20}
 {
  bone1Name=base_tasslesStart_r
  bone2Name=clavicle_r
  swing1LimitDegrees=45
  swing1Motion=Free
  swing2LimitDegrees=45
  swing2Motion=Free
  twistLimitDegrees=45
  twistMotion=Free
 }
 {bone1Name=dyn_tasslesStart_r, bone2Name=base_tasslesStart_r, swing1LimitDegrees=20, swing2LimitDegrees=0, twistLimitDegrees=0}
 {
  bone1Name=base_tassles_A_01_r
  bone2Name=dyn_tasslesStart_r
  swing1LimitDegrees=45
  swing1Motion=Free
  swing2LimitDegrees=45
  swing2Motion=Free
  twistLimitDegrees=45
  twistMotion=Free
 }
 {bone1Name=dyn_tassles_A_01_r, bone2Name=base_tassles_A_01_r, swing1LimitDegrees=50, swing2LimitDegrees=15, twistLimitDegrees=5}
 {bone1Name=dyn_tassles_A_02_r, bone2Name=dyn_tassles_A_01_r, swing1LimitDegrees=30, swing2LimitDegrees=15, twistLimitDegrees=20}
 {
  bone1Name=base_tassles_B_01_r
  bone2Name=dyn_tasslesStart_r
  swing1LimitDegrees=45
  swing1Motion=Free
  swing2LimitDegrees=45
  swing2Motion=Free
  twistLimitDegrees=45
  twistMotion=Free
 }
 {bone1Name=dyn_tassles_B_01_r, bone2Name=base_tassles_B_01_r, swing1LimitDegrees=50, swing2LimitDegrees=15, twistLimitDegrees=5}
 {bone1Name=dyn_tassles_B_02_r, bone2Name=dyn_tassles_B_01_r, swing1LimitDegrees=30, swing2LimitDegrees=15, twistLimitDegrees=20}
 {
  bone1Name=base_tassles_C_01_r
  bone2Name=dyn_tasslesStart_r
  swing1LimitDegrees=45
  swing1Motion=Free
  swing2LimitDegrees=45
  swing2Motion=Free
  twistLimitDegrees=45
  twistMotion=Free
 }
 {bone1Name=dyn_tassles_C_01_r, bone2Name=base_tassles_C_01_r, swing1LimitDegrees=50, swing2LimitDegrees=15, twistLimitDegrees=5}
 {bone1Name=dyn_tassles_C_02_r, bone2Name=dyn_tassles_C_01_r, swing1LimitDegrees=30, swing2LimitDegrees=15, twistLimitDegrees=20}
]
