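# Animation Blueprint for the TacticalPrince male body cosmetic, adding secondary motion to a pose copied from the parent mesh.
# A PREDeform Control Rig feeds additive jacket-flap offsets, then RigidBody and AnimDynamics chains simulate tassels, lapels, medals, flaps, jewelry, chains and pouches.
# A POSTDeform Control Rig produces the final pose, receiving a celebration-emote flag.
# The EventGraph updates gravity, simulation-space, movement and weapon state, plus an LOD-driven shrink morph.
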AnimBlueprint /BRCosmetics/Characters/Player/Male/Medium/Bodies/M_MED_TacticalPrince/Meshes/M_MED_TacticalPrince_AnimBP
target_skeleton=&/M_MED_TacticalPrince_Skeleton
graphs [
 AnimGraph {nodes [
   n1 = Control Rig!(Source=Use cached pose 'DynamicsCache'(), Alpha, Pub_ISCelebreteEmote:double=Get IsCelebrationEmote() -> (IsCelebrationEmote:double)) -> (Pose) {
    alphaBoolBlend {blendOption=Linear}
    controlRigAssetReference {blueprintRigClass=&/M_MED_TacticalPrince_POSTDeform_CtrlRig.M_MED_TacticalPrince_POSTDeform_CtrlRig_C}
   }
   Output Pose(Result=@n1) {layerGroup=None}
   n2 = Control Rig!(
    Source=Copy Pose From Mesh() {bUseAttachedParent=true}
    Alpha
    isAimingDownSights:double=0
    isCrouchingf:double=0
    isPickaxeOut:bool=false
    isSwingingPickaxe:double=0
    isPistolOut:bool=false
    Pub_IsCelebrationEmote:double=0
    Pub_CurrentLOD:bool=false
   ) -> (Pose) {
    alphaBoolBlend {blendOption=Linear}
    controlRigAssetReference {blueprintRigClass=&/M_MED_TacticalPrince_PREDeform_CtrlRig.M_MED_TacticalPrince_PREDeform_CtrlRig_C}
   }
   Save cached pose 'CRCache_Input'(Pose=@n2) {cacheName=CRCache_Input}
   n3 = RigidBody!(ComponentPose=Local To Component(LocalPose=Use cached pose 'PreDynamics_Cache'()), Alpha, OverrideWorldGravity:Vector="0,0,0") -> (Pose) {
    bOverrideWorldGravity=true
    overridePhysicsAsset=&/M_MED_TacticalPrince_Tassles_Physics
    simulationSpace=BaseBoneSpace
    alphaBoolBlend {blendOption=Linear}
    baseBoneRef {boneName=spine_05}
    simSpaceSettings {
     dampingAlpha=0
     maxAngularAcceleration=1600
     maxAngularVelocity=8
     maxLinearAcceleration=1600
     maxLinearVelocity=800
     velocityScaleZ=0.4
     worldAlpha=1
     externalLinearDragV=<0.2 0.2 0.2>
    }
   }
   n4 = RigidBody(ComponentPose=@n3, OverrideWorldGravity="0,0,0") {
    bOverrideWorldGravity=true
    overridePhysicsAsset=&/M_MED_TacticalPrince_Lapels_Physics
    simulationSpace=BaseBoneSpace
    alphaBoolBlend {blendOption=Linear}
    baseBoneRef {boneName=spine_05}
    componentAppliedLinearAccClamp=<3000 3000 3000>
    componentLinearAccScale=<0.1 0.1 0.1>
    componentLinearVelScale=<0.1 0.1 0.1>
    simSpaceSettings {
     maxAngularAcceleration=1000
     maxAngularVelocity=3
     maxLinearAcceleration=700
     maxLinearVelocity=500
     worldAlpha=1
     externalLinearDragV {x=0.2}
    }
   }
   n5 = RigidBody(ComponentPose=@n4, OverrideWorldGravity="0,0,0") {
    bOverrideWorldGravity=true
    overridePhysicsAsset=&/M_MED_TacticalPrince_Chains_Physics
    simulationSpace=BaseBoneSpace
    alphaBoolBlend {blendOption=Linear}
    baseBoneRef {boneName=root}
    componentAppliedLinearAccClamp=<3000 3000 3000>
    simSpaceSettings {
     maxAngularAcceleration=1
     maxAngularVelocity=3
     maxLinearAcceleration=1000
     maxLinearVelocity=500
     velocityScaleZ=2
     worldAlpha=1
     externalLinearDragV=<0.5 0.5 0.5>
    }
   }
   n6 = "Constraint - base_medal_01 (PPP)"!(ComponentPose=@n5, Alpha=1, ConstraintWeights_0=0.0151, ConstraintWeights_1:float=0.2709, ConstraintWeights_2:float=0.7139) -> (Pose) {
    alphaBoolBlend {blendOption=Linear}
    boneToModify {boneName=base_medal_01}
    constraintSetup [{targetBone {boneName=dyn_neckLapel_01_r}}, {targetBone {boneName=dyn_jacktLapel_r}}, {targetBone {boneName=dyn_jacktLapel_low_r}}]
    constraintWeights [1,0,1]
   }
   n7 = Anim Dynamics - Bone: dyn_medalStar_02!(SimulationSpace:AnimPhysSimSpaceType Enum, ComponentPose=@n6, Alpha=1, GravityOverride:Vector="0,0,-980") -> (Pose) {
    angularDampingOverride=1
    angularSpringConstant=20000
    bOverrideAngularDamping=true
    bUseGravityOverride=true
    gravityScale=0.35
    simulationSpace=RootRelative
    alphaBoolBlend {blendOption=Linear}
    boundBone {boneName=dyn_medalStar_02}
    componentAppliedLinearAccClamp=<10000 10000 10000>
    componentLinearAccScale=<1 1 1>
    componentLinearVelScale=<1 1 1>
    gravityOverride {z=-980}
    physicsBodyDefinitions [{
      boundBone {boneName=dyn_medalStar_02}
      boxExtents=<10 16 16>
      constraintSetup {angularLimitsMax=<0 25 0>, angularLimitsMin=<0 -25 -45>, angularTarget=<1 0 0>}
      localJointOffset=<1.5 0 0>
     }]
   }
   n8 = Anim Dynamics - Bone: dyn_medalPendant_03!(SimulationSpace:AnimPhysSimSpaceType Enum, ComponentPose=@n7, Alpha=1, GravityOverride:Vector="0,0,-980") -> (Pose) {
    angularDampingOverride=1
    angularSpringConstant=20000
    bOverrideAngularDamping=true
    bUseGravityOverride=true
    gravityScale=0.35
    simulationSpace=RootRelative
    alphaBoolBlend {blendOption=Linear}
    boundBone {boneName=dyn_medalPendant_03}
    componentAppliedLinearAccClamp=<10000 10000 10000>
    componentLinearAccScale=<1 1 1>
    componentLinearVelScale=<1 1 1>
    gravityOverride {z=-980}
    physicsBodyDefinitions [{
      boundBone {boneName=dyn_medalPendant_03}
      boxExtents=<10 16 16>
      constraintSetup {angularLimitsMax=<0 25 0>, angularLimitsMin=<0 -25 -45>, angularTarget=<1 0 0>}
      localJointOffset=<1.7 0 0>
     }]
   }
   n9 = Anim Dynamics - Bone: dyn_medalString_04!(SimulationSpace:AnimPhysSimSpaceType Enum, ComponentPose=@n8, Alpha=1, GravityOverride:Vector="0,0,-980") -> (Pose) {
    angularDampingOverride=1
    angularSpringConstant=1800
    bAngularSpring=true
    bOverrideAngularDamping=true
    bUseGravityOverride=true
    gravityScale=0.35
    simulationSpace=RootRelative
    alphaBoolBlend {blendOption=Linear}
    boundBone {boneName=dyn_medalString_04}
    componentAppliedLinearAccClamp=<10000 10000 10000>
    componentLinearAccScale=<1 1 1>
    componentLinearVelScale=<1 1 1>
    gravityOverride {z=-980}
    physicsBodyDefinitions [{
      boundBone {boneName=dyn_medalString_04}
      boxExtents=<12 18 18>
      constraintSetup {angularLimitsMax=<0 25 0>, angularLimitsMin=<0 -25 -45>, angularTarget=<1 0 0>}
      localJointOffset=<1.7 0 0>
     }]
   }
   n10 = Anim Dynamics - Bone: dyn_jacketFlap_Front_01_l!(SimulationSpace:AnimPhysSimSpaceType Enum=World, ComponentPose=@n9, Alpha=1, GravityOverride:Vector="0,0,-980") -> (Pose) {
    angularDampingOverride=1
    angularSpringConstant=6000
    bAngularSpring=true
    bOverrideAngularDamping=true
    bUseGravityOverride=true
    gravityScale=0.35
    simulationSpace=World
    alphaBoolBlend {blendOption=Linear}
    boundBone {boneName=dyn_jacketFlap_Front_01_l}
    componentAppliedLinearAccClamp=<10000 10000 10000>
    componentLinearAccScale=<1 1 1>
    componentLinearVelScale=<1 1 1>
    gravityOverride {z=-980}
    physicsBodyDefinitions [{
      boundBone {boneName=dyn_jacketFlap_Front_01_l}
      boxExtents=<25 45 35>
      constraintSetup {angularLimitsMax=<2 10 0>, angularLimitsMin=<-2 -10 -35>, angularTarget=<1 0 0>}
      localJointOffset=<3 0 0>
     }]
   }
   n11 = Anim Dynamics - Bone: dyn_jacketFlap_Front_02_l!(SimulationSpace:AnimPhysSimSpaceType Enum=World, ComponentPose=@n10, Alpha=1, GravityOverride:Vector="0,0,-980") -> (Pose) {
    angularDampingOverride=1
    angularSpringConstant=6000
    bAngularSpring=true
    bOverrideAngularDamping=true
    bUseGravityOverride=true
    gravityScale=0.35
    simulationSpace=World
    alphaBoolBlend {blendOption=Linear}
    boundBone {boneName=dyn_jacketFlap_Front_02_l}
    componentAppliedLinearAccClamp=<10000 10000 10000>
    componentLinearAccScale=<1 1 1>
    componentLinearVelScale=<1 1 1>
    gravityOverride {z=-980}
    physicsBodyDefinitions [{
      boundBone {boneName=dyn_jacketFlap_Front_02_l}
      boxExtents=<20 40 30>
      constraintSetup {angularLimitsMax=<2 10 0>, angularLimitsMin=<-2 -10 -25>, angularTarget=<1 0 0>}
      localJointOffset=<3 0 0>
     }]
   }
   n12 = Anim Dynamics - Bone: dyn_jacketFlap_Side_01_l!(SimulationSpace:AnimPhysSimSpaceType Enum=World, ComponentPose=@n11, Alpha=1, GravityOverride:Vector="0,0,-980") -> (Pose) {
    angularDampingOverride=1
    angularSpringConstant=6000
    bAngularSpring=true
    bOverrideAngularDamping=true
    bUseGravityOverride=true
    gravityScale=0.35
    simulationSpace=World
    alphaBoolBlend {blendOption=Linear}
    boundBone {boneName=dyn_jacketFlap_Side_01_l}
    componentAppliedLinearAccClamp=<10000 10000 10000>
    componentLinearAccScale=<1 1 1>
    componentLinearVelScale=<1 1 1>
    gravityOverride {z=-980}
    physicsBodyDefinitions [{
      boundBone {boneName=dyn_jacketFlap_Side_01_l}
      boxExtents=<25 50 35>
      constraintSetup {angularLimitsMax=<2 10 0>, angularLimitsMin=<-2 -10 -35>, angularTarget=<1 0 0>}
      localJointOffset=<3 0 0>
     }]
   }
   n13 = Anim Dynamics - Bone: dyn_jacketFlap_Side_02_l!(SimulationSpace:AnimPhysSimSpaceType Enum=World, ComponentPose=@n12, Alpha=1, GravityOverride:Vector="0,0,-980") -> (Pose) {
    angularDampingOverride=1
    angularSpringConstant=6000
    bAngularSpring=true
    bOverrideAngularDamping=true
    bUseGravityOverride=true
    gravityScale=0.35
    simulationSpace=World
    alphaBoolBlend {blendOption=Linear}
    boundBone {boneName=dyn_jacketFlap_Side_02_l}
    componentAppliedLinearAccClamp=<10000 10000 10000>
    componentLinearAccScale=<1 1 1>
    componentLinearVelScale=<1 1 1>
    gravityOverride {z=-980}
    physicsBodyDefinitions [{
      boundBone {boneName=dyn_jacketFlap_Side_02_l}
      boxExtents=<20 45 30>
      constraintSetup {angularLimitsMax=<2 10 0>, angularLimitsMin=<-2 -10 -25>, angularTarget=<1 0 0>}
      localJointOffset=<3 0 0>
     }]
   }
   n14 = Anim Dynamics - Bone: dyn_jacketFlap_Back_01_l!(SimulationSpace:AnimPhysSimSpaceType Enum=World, ComponentPose=@n13, Alpha=1, GravityOverride:Vector="0,0,-980") -> (Pose) {
    angularDampingOverride=1
    angularSpringConstant=6000
    bAngularSpring=true
    bOverrideAngularDamping=true
    bUseGravityOverride=true
    gravityScale=0.35
    simulationSpace=World
    alphaBoolBlend {blendOption=Linear}
    boundBone {boneName=dyn_jacketFlap_Back_01_l}
    componentAppliedLinearAccClamp=<10000 10000 10000>
    componentLinearAccScale=<1 1 1>
    componentLinearVelScale=<1 1 1>
    gravityOverride {z=-980}
    physicsBodyDefinitions [{
      boundBone {boneName=dyn_jacketFlap_Back_01_l}
      boxExtents=<25 45 35>
      constraintSetup {angularLimitsMax=<2 10 0>, angularLimitsMin=<-2 -10 -30>, angularTarget=<1 0 0>}
      localJointOffset=<3 0 0>
     }]
   }
   n15 = Anim Dynamics - Bone: dyn_jacketFlap_Back_02_l!(SimulationSpace:AnimPhysSimSpaceType Enum=World, ComponentPose=@n14, Alpha=1, GravityOverride:Vector="0,0,-980") -> (Pose) {
    angularDampingOverride=1
    angularSpringConstant=6000
    bAngularSpring=true
    bOverrideAngularDamping=true
    bUseGravityOverride=true
    gravityScale=0.35
    simulationSpace=World
    alphaBoolBlend {blendOption=Linear}
    boundBone {boneName=dyn_jacketFlap_Back_02_l}
    componentAppliedLinearAccClamp=<10000 10000 10000>
    componentLinearAccScale=<1 1 1>
    componentLinearVelScale=<1 1 1>
    gravityOverride {z=-980}
    physicsBodyDefinitions [{
      boundBone {boneName=dyn_jacketFlap_Back_02_l}
      boxExtents=<20 40 30>
      constraintSetup {angularLimitsMax=<2 10 0>, angularLimitsMin=<-2 -10 -25>, angularTarget=<1 0 0>}
      localJointOffset=<3 0 0>
     }]
   }
   n16 = Anim Dynamics - Bone: dyn_jacketFlap_Back_01_r!(SimulationSpace:AnimPhysSimSpaceType Enum=World, ComponentPose=@n15, Alpha=1, GravityOverride:Vector="0,0,-980") -> (Pose) {
    angularDampingOverride=1
    angularSpringConstant=6000
    bAngularSpring=true
    bOverrideAngularDamping=true
    bUseGravityOverride=true
    gravityScale=0.35
    simulationSpace=World
    alphaBoolBlend {blendOption=Linear}
    boundBone {boneName=dyn_jacketFlap_Back_01_r}
    componentAppliedLinearAccClamp=<10000 10000 10000>
    componentLinearAccScale=<1 1 1>
    componentLinearVelScale=<1 1 1>
    gravityOverride {z=-980}
    physicsBodyDefinitions [{
      boundBone {boneName=dyn_jacketFlap_Back_01_r}
      boxExtents=<25 45 35>
      constraintSetup {angularLimitsMax=<2 10 0>, angularLimitsMin=<-2 -10 -30>, angularTarget=<1 0 0>}
      localJointOffset=<3 0 0>
     }]
   }
   n17 = Anim Dynamics - Bone: dyn_jacketFlap_Back_02_r!(SimulationSpace:AnimPhysSimSpaceType Enum=World, ComponentPose=@n16, Alpha=1, GravityOverride:Vector="0,0,-980") -> (Pose) {
    angularDampingOverride=1
    angularSpringConstant=6000
    bAngularSpring=true
    bOverrideAngularDamping=true
    bUseGravityOverride=true
    gravityScale=0.35
    simulationSpace=World
    alphaBoolBlend {blendOption=Linear}
    boundBone {boneName=dyn_jacketFlap_Back_02_r}
    componentAppliedLinearAccClamp=<10000 10000 10000>
    componentLinearAccScale=<1 1 1>
    componentLinearVelScale=<1 1 1>
    gravityOverride {z=-980}
    physicsBodyDefinitions [{
      boundBone {boneName=dyn_jacketFlap_Back_02_r}
      boxExtents=<20 40 30>
      constraintSetup {angularLimitsMax=<2 10 0>, angularLimitsMin=<-2 -10 -25>, angularTarget=<1 0 0>}
      localJointOffset=<3 0 0>
     }]
   }
   n18 = Anim Dynamics - Bone: dyn_jacketFlap_Side_01_r!(SimulationSpace:AnimPhysSimSpaceType Enum=World, ComponentPose=@n17, Alpha=1, GravityOverride:Vector="0,0,-980") -> (Pose) {
    angularDampingOverride=1
    angularSpringConstant=6000
    bAngularSpring=true
    bOverrideAngularDamping=true
    bUseGravityOverride=true
    gravityScale=0.35
    simulationSpace=World
    alphaBoolBlend {blendOption=Linear}
    boundBone {boneName=dyn_jacketFlap_Side_01_r}
    componentAppliedLinearAccClamp=<10000 10000 10000>
    componentLinearAccScale=<1 1 1>
    componentLinearVelScale=<1 1 1>
    gravityOverride {z=-980}
    physicsBodyDefinitions [{
      boundBone {boneName=dyn_jacketFlap_Side_01_r}
      boxExtents=<25 50 35>
      constraintSetup {angularLimitsMax=<2 10 0>, angularLimitsMin=<-2 -10 -35>, angularTarget=<1 0 0>}
      localJointOffset=<3 0 0>
     }]
   }
   n19 = Anim Dynamics - Bone: dyn_jacketFlap_Side_02_r!(SimulationSpace:AnimPhysSimSpaceType Enum=World, ComponentPose=@n18, Alpha=1, GravityOverride:Vector="0,0,-980") -> (Pose) {
    angularDampingOverride=1
    angularSpringConstant=6000
    bAngularSpring=true
    bOverrideAngularDamping=true
    bUseGravityOverride=true
    gravityScale=0.35
    simulationSpace=World
    alphaBoolBlend {blendOption=Linear}
    boundBone {boneName=dyn_jacketFlap_Side_02_r}
    componentAppliedLinearAccClamp=<10000 10000 10000>
    componentLinearAccScale=<1 1 1>
    componentLinearVelScale=<1 1 1>
    gravityOverride {z=-980}
    physicsBodyDefinitions [{
      boundBone {boneName=dyn_jacketFlap_Side_02_r}
      boxExtents=<20 45 30>
      constraintSetup {angularLimitsMax=<2 10 0>, angularLimitsMin=<-2 -10 -25>, angularTarget=<1 0 0>}
      localJointOffset=<3 0 0>
     }]
   }
   n20 = Anim Dynamics - Bone: dyn_jacketFlap_Front_01_r!(SimulationSpace:AnimPhysSimSpaceType Enum=World, ComponentPose=@n19, Alpha=1, GravityOverride:Vector="0,0,-980") -> (Pose) {
    angularDampingOverride=1
    angularSpringConstant=6000
    bAngularSpring=true
    bOverrideAngularDamping=true
    bUseGravityOverride=true
    gravityScale=0.35
    simulationSpace=World
    alphaBoolBlend {blendOption=Linear}
    boundBone {boneName=dyn_jacketFlap_Front_01_r}
    componentAppliedLinearAccClamp=<10000 10000 10000>
    componentLinearAccScale=<1 1 1>
    componentLinearVelScale=<1 1 1>
    gravityOverride {z=-980}
    physicsBodyDefinitions [{
      boundBone {boneName=dyn_jacketFlap_Front_01_r}
      boxExtents=<25 45 35>
      constraintSetup {angularLimitsMax=<2 10 0>, angularLimitsMin=<-2 -10 -35>, angularTarget=<1 0 0>}
      localJointOffset=<3 0 0>
     }]
   }
   n21 = Anim Dynamics - Bone: dyn_jacketFlap_Front_02_r!(SimulationSpace:AnimPhysSimSpaceType Enum=World, ComponentPose=@n20, Alpha=1, GravityOverride:Vector="0,0,-980") -> (Pose) {
    angularDampingOverride=1
    angularSpringConstant=6000
    bAngularSpring=true
    bOverrideAngularDamping=true
    bUseGravityOverride=true
    gravityScale=0.35
    simulationSpace=World
    alphaBoolBlend {blendOption=Linear}
    boundBone {boneName=dyn_jacketFlap_Front_02_r}
    componentAppliedLinearAccClamp=<10000 10000 10000>
    componentLinearAccScale=<1 1 1>
    componentLinearVelScale=<1 1 1>
    gravityOverride {z=-980}
    physicsBodyDefinitions [{
      boundBone {boneName=dyn_jacketFlap_Front_02_r}
      boxExtents=<20 40 30>
      constraintSetup {angularLimitsMax=<2 10 0>, angularLimitsMin=<-2 -10 -25>, angularTarget=<1 0 0>}
      localJointOffset=<3 0 0>
     }]
   }
   n22 = "Transform (Modify) Bone - Bone: base_necklace_01"!(ComponentPose=@n21, bAlphaBoolEnabled:bool=True, Rotation="0,0,-2") -> (Pose) {
    alphaInputType=Bool
    rotationMode=BMM_Additive
    translationSpace=BCS_BoneSpace
    alphaBoolBlend {blendOption=Linear}
    boneToModify {boneName=base_necklace_01}
    rotation {roll=-2}
   }
   n23 = Anim Dynamics - Bone: dyn_necklace_01!(SimulationSpace:AnimPhysSimSpaceType Enum, ComponentPose=@n22, Alpha=1, GravityOverride:Vector="0,0,-980") -> (Pose) {
    angularDampingOverride=0.9
    angularSpringConstant=1000
    bAngularSpring=true
    bOverrideAngularDamping=true
    bUseGravityOverride=true
    gravityScale=0.35
    simulationSpace=RootRelative
    alphaBoolBlend {blendOption=Linear}
    boundBone {boneName=dyn_necklace_01}
    componentAppliedLinearAccClamp=<10000 10000 10000>
    componentLinearAccScale=<1 1 1>
    componentLinearVelScale=<1 1 1>
    gravityOverride {z=-980}
    physicsBodyDefinitions [{
      boundBone {boneName=dyn_necklace_01}
      boxExtents=<8 12 12>
      constraintSetup {angularLimitsMax=<0 0 0>, angularLimitsMin=<0 0 -40>, angularTarget=<1 0 0>}
      localJointOffset=<1 0 0>
     }]
   }
   n24 = Anim Dynamics - Bone: dyn_necklace_02!(SimulationSpace:AnimPhysSimSpaceType Enum, ComponentPose=@n23, Alpha=1, GravityOverride:Vector="0,0,-980") -> (Pose) {
    angularDampingOverride=0.9
    angularSpringConstant=1000
    bOverrideAngularDamping=true
    bUseGravityOverride=true
    gravityScale=0.35
    simulationSpace=RootRelative
    alphaBoolBlend {blendOption=Linear}
    boundBone {boneName=dyn_necklace_02}
    componentAppliedLinearAccClamp=<10000 10000 10000>
    componentLinearAccScale=<1 1 1>
    componentLinearVelScale=<1 1 1>
    gravityOverride {z=-980}
    physicsBodyDefinitions [{
      boundBone {boneName=dyn_necklace_02}
      boxExtents=<10 15 12>
      constraintSetup {angularLimitsMax=<0 10 0>, angularLimitsMin=<0 -10 -25>, angularTarget=<1 0 0>}
      localJointOffset=<0.2 0 0>
     }]
    sphericalLimits [{limitRadius=13, drivingBone {boneName=spine_05}, sphereLocalOffset=<0 -2 0>}]
   }
   n25 = Anim Dynamics - Bone: dyn_dogTag_01!(SimulationSpace:AnimPhysSimSpaceType Enum, ComponentPose=@n24, Alpha=1, GravityOverride:Vector="0,0,-980") -> (Pose) {
    angularDampingOverride=1
    angularSpringConstant=20000
    bOverrideAngularDamping=true
    bUseGravityOverride=true
    gravityScale=0.35
    simulationSpace=RootRelative
    alphaBoolBlend {blendOption=Linear}
    boundBone {boneName=dyn_dogTag_01}
    componentAppliedLinearAccClamp=<3000 3000 3000>
    componentLinearAccScale=<0.6 1 0.6>
    componentLinearVelScale=<0.6 1 0.6>
    gravityOverride {z=-980}
    physicsBodyDefinitions [{
      boundBone {boneName=dyn_dogTag_01}
      boxExtents=<10 16 8>
      constraintSetup {angularLimitsMax=<5 25 0>, angularLimitsMin=<-5 -25 -20>, angularTarget=<0 0 0>}
      localJointOffset=<4 0 0>
     }]
   }
   n26 = Anim Dynamics - Bone: dyn_pouch_02!(SimulationSpace:AnimPhysSimSpaceType Enum=World, ComponentPose=@n25, Alpha=1, GravityOverride:Vector="0,0,-980") -> (Pose) {
    angularDampingOverride=1
    angularSpringConstant=20000
    bAngularSpring=true
    bOverrideAngularDamping=true
    bUseGravityOverride=true
    gravityScale=0.35
    simulationSpace=World
    alphaBoolBlend {blendOption=Linear}
    boundBone {boneName=dyn_pouch_02}
    componentAppliedLinearAccClamp=<10000 10000 10000>
    componentLinearAccScale=<1 1 1>
    componentLinearVelScale=<1 1 1>
    gravityOverride {z=-980}
    physicsBodyDefinitions [{
      boundBone {boneName=dyn_pouch_02}
      boxExtents=<30 50 50>
      constraintSetup {angularLimitsMax=<10 10 0>, angularLimitsMin=<-10 -10 -15>, angularTarget=<1 0 0>}
      localJointOffset=<2 0 0>
     }]
   }
   n27 = Anim Dynamics - Bone: dyn_pouch_01!(SimulationSpace:AnimPhysSimSpaceType Enum=World, ComponentPose=@n26, Alpha=1, GravityOverride:Vector="0,0,-980") -> (Pose) {
    angularDampingOverride=1
    angularSpringConstant=20000
    bAngularSpring=true
    bOverrideAngularDamping=true
    bUseGravityOverride=true
    gravityScale=0.35
    simulationSpace=World
    alphaBoolBlend {blendOption=Linear}
    boundBone {boneName=dyn_pouch_01}
    componentAppliedLinearAccClamp=<10000 10000 10000>
    componentLinearAccScale=<1 1 1>
    componentLinearVelScale=<1 1 1>
    gravityOverride {z=-980}
    physicsBodyDefinitions [{
      boundBone {boneName=dyn_pouch_01}
      boxExtents=<30 50 50>
      constraintSetup {angularLimitsMax=<10 10 0>, angularLimitsMin=<-10 -10 -15>, angularTarget=<1 0 0>}
      localJointOffset=<2 0 0>
     }]
   }
   n28 = Anim Dynamics - Bone: dyn_legPouch_01!(SimulationSpace:AnimPhysSimSpaceType Enum=World, ComponentPose=@n27, Alpha=1, GravityOverride:Vector="0,0,-980") -> (Pose) {
    angularDampingOverride=0.9999
    angularSpringConstant=20000
    bAngularSpring=true
    bOverrideAngularDamping=true
    bUseGravityOverride=true
    gravityScale=0.35
    simulationSpace=World
    alphaBoolBlend {blendOption=Linear}
    boundBone {boneName=dyn_legPouch_01}
    componentAppliedLinearAccClamp=<10000 10000 10000>
    componentLinearAccScale=<1 1 1>
    componentLinearVelScale=<1 1 1>
    gravityOverride {z=-980}
    physicsBodyDefinitions [{
      boundBone {boneName=dyn_legPouch_01}
      boxExtents=<30 40 60>
      constraintSetup {angularLimitsMax=<20 3 0>, angularLimitsMin=<-20 -3 -15>, angularTarget=<1 0 0>}
      localJointOffset=<0.8 0 0>
     }]
   }
   n29 = Anim Dynamics - Bone: dyn_frontChain_01!(SimulationSpace:AnimPhysSimSpaceType Enum, ComponentPose=@n28, Alpha=1, GravityOverride:Vector="0,0,-980") -> (Pose) {
    angularDampingOverride=1
    angularSpringConstant=500
    bOverrideAngularDamping=true
    bUseGravityOverride=true
    gravityScale=0.35
    simulationSpace=RootRelative
    alphaBoolBlend {blendOption=Linear}
    boundBone {boneName=dyn_frontChain_01}
    componentAppliedLinearAccClamp=<10000 10000 10000>
    componentLinearAccScale=<1 1 1>
    componentLinearVelScale=<1 1 1>
    gravityOverride {z=-980}
    physicsBodyDefinitions [{
      boundBone {boneName=dyn_frontChain_01}
      boxExtents=<8 16 16>
      constraintSetup {angularLimitsMax=<0 5 0>, angularLimitsMin=<0 -5 -50>, angularTarget=<1 0 0>}
      localJointOffset=<2 0 0>
     }]
   }
   n30 = Anim Dynamics - Bone: dyn_frontChain_02!(SimulationSpace:AnimPhysSimSpaceType Enum, ComponentPose=@n29, Alpha=1, GravityOverride:Vector="0,0,-980") -> (Pose) {
    angularDampingOverride=1
    angularSpringConstant=500
    bOverrideAngularDamping=true
    bUseGravityOverride=true
    gravityScale=0.35
    simulationSpace=RootRelative
    alphaBoolBlend {blendOption=Linear}
    boundBone {boneName=dyn_frontChain_02}
    componentAppliedLinearAccClamp=<10000 10000 10000>
    componentLinearAccScale=<1 1 1>
    componentLinearVelScale=<1 1 1>
    gravityOverride {z=-980}
    physicsBodyDefinitions [{
      boundBone {boneName=dyn_frontChain_02}
      boxExtents=<7 14 14>
      constraintSetup {angularLimitsMax=<0 5 0>, angularLimitsMin=<0 -5 -50>, angularTarget=<1 0 0>}
      localJointOffset=<3 0 0>
     }]
   }
   n31 = Anim Dynamics - Bone: dyn_frontChain_03!(SimulationSpace:AnimPhysSimSpaceType Enum, ComponentPose=@n30, Alpha=1, GravityOverride:Vector="0,0,-980") -> (Pose) {
    angularDampingOverride=1
    angularSpringConstant=500
    bOverrideAngularDamping=true
    bUseGravityOverride=true
    gravityScale=0.35
    simulationSpace=RootRelative
    alphaBoolBlend {blendOption=Linear}
    boundBone {boneName=dyn_frontChain_03}
    componentAppliedLinearAccClamp=<10000 10000 10000>
    componentLinearAccScale=<1 1 1>
    componentLinearVelScale=<1 1 1>
    gravityOverride {z=-980}
    physicsBodyDefinitions [{
      boundBone {boneName=dyn_frontChain_03}
      boxExtents=<7 14 14>
      constraintSetup {angularLimitsMax=<0 5 0>, angularLimitsMin=<0 -5 -30>, angularTarget=<1 0 0>}
      localJointOffset=<3 0 0>
     }]
   }
   n32 = Anim Dynamics - Bone: dyn_backChain_01!(SimulationSpace:AnimPhysSimSpaceType Enum=World, ComponentPose=@n31, Alpha=1, GravityOverride:Vector="0,0,-980") -> (Pose) {
    angularDampingOverride=1
    angularSpringConstant=500
    bOverrideAngularDamping=true
    bUseGravityOverride=true
    gravityScale=0.35
    simulationSpace=World
    alphaBoolBlend {blendOption=Linear}
    boundBone {boneName=dyn_backChain_01}
    componentAppliedLinearAccClamp=<10000 10000 10000>
    componentLinearAccScale=<1 1 1>
    componentLinearVelScale=<1 1 1>
    gravityOverride {z=-980}
    physicsBodyDefinitions [{
      boundBone {boneName=dyn_backChain_01}
      boxExtents=<7 14 14>
      constraintSetup {angularLimitsMax=<0 5 0>, angularLimitsMin=<0 -5 -50>, angularTarget=<1 0 0>}
      localJointOffset=<3 0 0>
     }]
   }
   n33 = Anim Dynamics - Bone: dyn_backChain_02!(SimulationSpace:AnimPhysSimSpaceType Enum=World, ComponentPose=@n32, Alpha=1, GravityOverride:Vector="0,0,-980") -> (Pose) {
    angularDampingOverride=1
    angularSpringConstant=500
    bOverrideAngularDamping=true
    bUseGravityOverride=true
    gravityScale=0.35
    simulationSpace=World
    alphaBoolBlend {blendOption=Linear}
    boundBone {boneName=dyn_backChain_02}
    componentAppliedLinearAccClamp=<10000 10000 10000>
    componentLinearAccScale=<1 1 1>
    componentLinearVelScale=<1 1 1>
    gravityOverride {z=-980}
    physicsBodyDefinitions [{
      boundBone {boneName=dyn_backChain_02}
      boxExtents=<7 14 14>
      constraintSetup {angularLimitsMax=<0 5 0>, angularLimitsMin=<0 -5 -50>, angularTarget=<1 0 0>}
      localJointOffset=<3 0 0>
     }]
   }
   Save cached pose 'DynamicsCache'(Pose=Component To Local(ComponentPose=@n33)) {cacheName=DynamicsCache}
   n34 = "Transform (Modify) Bone - Bone: base_jacketFlap_Side_01_l"!(
    ComponentPose=Local To Component(LocalPose=Use cached pose 'CRCache_Input'())
    bAlphaBoolEnabled:bool=True
    Translation="0,1,0"
    Rotation="0,0,8"
   ) -> (Pose) {
    alphaInputType=Bool
    rotationMode=BMM_Additive
    translationMode=BMM_Additive
    alphaBoolBlend {blendInTime=0.2, blendOption=Linear, blendOutTime=0.25}
    boneToModify {boneName=base_jacketFlap_Side_01_l}
    rotation {roll=8}
    translation {y=1}
   }
   n35 = "Transform (Modify) Bone - Bone: base_jacketFlap_Back_01_l"!(ComponentPose=@n34, bAlphaBoolEnabled:bool=True, Translation="0,2,0", Rotation="0,0,10") -> (Pose) {
    alphaInputType=Bool
    rotationMode=BMM_Additive
    translationMode=BMM_Additive
    alphaBoolBlend {blendInTime=0.2, blendOption=Linear, blendOutTime=0.25}
    boneToModify {boneName=base_jacketFlap_Back_01_l}
    rotation {roll=10}
    translation {y=2}
   }
   n36 = "Transform (Modify) Bone - Bone: base_jacketFlap_Back_01_r"!(ComponentPose=@n35, bAlphaBoolEnabled:bool=True, Translation="0,2,0", Rotation="0,0,10") -> (Pose) {
    alphaInputType=Bool
    rotationMode=BMM_Additive
    translationMode=BMM_Additive
    alphaBoolBlend {blendInTime=0.2, blendOption=Linear, blendOutTime=0.25}
    boneToModify {boneName=base_jacketFlap_Back_01_r}
    rotation {roll=10}
    translation {y=2}
   }
   n37 = "Transform (Modify) Bone - Bone: base_jacketFlap_Side_01_r"!(ComponentPose=@n36, bAlphaBoolEnabled:bool=True, Translation="0,1,0", Rotation="0,0,8") -> (Pose) {
    alphaInputType=Bool
    rotationMode=BMM_Additive
    translationMode=BMM_Additive
    alphaBoolBlend {blendInTime=0.2, blendOption=Linear, blendOutTime=0.25}
    boneToModify {boneName=base_jacketFlap_Side_01_r}
    rotation {roll=8}
    translation {y=1}
   }
   Save cached pose 'PreDynamics_Cache'(Pose=Component To Local(ComponentPose=@n37)) {cacheName=PreDynamics_Cache}
  ]}
 EventGraph {nodes [
   n1 = Event BlueprintUpdateAnimation()
   > n2 = Sequence!(execute) -> (then_0, then_1, then_2:Exec, then_3:Exec, then_4:Exec, then_5:Exec)
   > Set Delta Time X(Delta Time X=@n1.DeltaTimeX)
   > n3 = Gravity Override State Function DEPRECATED(CharacterPartAnimInstance=Self-Reference(), GameTimeInSeconds=GetGameTimeInSeconds(), GravityOverrideParams=Get Gravity Override Params())
   > Set GravityOverride(GravityOverride=@n3)
   n4 = Get bIsTargeting()
   > Set IsAimingDown(IsAimingDown:double=FInterpTo(Current=Get IsAimingDown() -> (IsAimingDown:double), Target="To Float (Boolean)"(InBool=@n4), DeltaTime=Get Delta Time X(), InterpSpeed=5)) -> (Output_Get:double)
   n5 = Get bIsCrouching()
   > Set isCrouchingf(isCrouchingf:double=FInterpTo(Current=Get isCrouchingf() -> (isCrouchingf:double), Target="To Float (Boolean)"(InBool=@n5), DeltaTime=Get Delta Time X(), InterpSpeed=1)) -> (Output_Get:double)
   n6 = Get isSwingingPickaxe() -> (isSwingingPickaxe:bool)
   > Set isSwingingf(isSwingingf:double=FInterpTo(Current=Get isSwingingf() -> (isSwingingf:double), Target="To Float (Boolean)"(InBool=@n6), DeltaTime=Get Delta Time X(), InterpSpeed=5)) -> (Output_Get:double)
   > n7 = Branch(Condition=Get bIsInFrontEnd())
   > Set SimlationSpace(SimlationSpace:AnimPhysSimSpaceType Enum=World) -> (Output_Get:AnimPhysSimSpaceType Enum)
   Set SimlationSpace(execute=@n7.else)
   n8 = Branch(
    execute=@n2.then_2
    Condition=OR Boolean!(A=Get bIsSkydiveFloating(), B=Get bIsParachuteOpened(), C:bool=Get bIsSkydiveDiveMode(), D:bool=Get bIsInFrontEnd()) -> (ReturnValue)
   )
   > Set AlphaFE(AlphaFE:double=0.4) -> (Output_Get:double)
   Set AlphaFE(execute=@n8.else, AlphaFE=1)
   Branch(execute=@n2.then_3, Condition="integer > integer"(A=Get CurrentLOD()))
   > Set ActivateExtraFixes(ActivateExtraFixes:bool=true) -> (Output_Get:bool)
   n9 = Branch(execute=@n2.then_4, Condition=Get bIsSkydiveDiveMode())
   Do Once(execute=@n9.else, Reset=@n9.then, Start Closed=true)
   > Set SkyDivingTransitionOn(SkyDivingTransitionOn:bool=true) -> (Output_Get:bool)
   > Delay(Duration=1.25)
   > Set SkyDivingTransitionOn(SkyDivingTransitionOn=false)
   n10 = GetAnimInstance(self=GetOwnerSkeletalMeshComponent())
   Is Valid(exec=@n2.then_1, InputObject=@n10)
   > n11 = Set Iscelebration(Iscelebration:bool=Montage_IsPlaying(self=@n10, Montage=/Game/Animation/Game/MainPlayer/Emotes/Emote_Celebration_M)) -> (Output_Get:bool)
   > n12 = Branch(Condition=@n11)
   > Set IsCelebrationEmote(IsCelebrationEmote:double=1) -> (Output_Get:double)
   Set IsCelebrationEmote(execute=@n12.else, IsCelebrationEmote=0)
   n13 = Branch(execute=@n2.then_5, Condition="integer > integer"(A=Get CurrentLOD(), B=1) {is=title K2Node_CallFunction})
   > SetMorphTarget!(execute, self:Anim Instance Object Reference, MorphTargetName:Name=Shrink_LODs_MorphShape, Value:float=1) -> (then)
   SetMorphTarget(execute=@n13.else, MorphTargetName=Shrink_LODs_MorphShape, Value=0)
   Event OnBeginPickaxeSwing() {is=title K2Node_Event}
   > Set isSwingingPickaxe(isSwingingPickaxe:bool=true) -> (Output_Get:bool)
   Event OnWeaponChanged()
   > Set PistolOut(PistolOut:bool=false) -> (Output_Get:bool)
   > Set PickAxeOut(PickAxeOut:bool=false) -> (Output_Get:bool)
   > n14 = Switch on EFortWeaponCoreAnimation!(execute, Selection:EFortWeaponCoreAnimation Enum=Get WeaponCoreAnim!() -> (WeaponCoreAnim)) -> (
    Melee:Exec
    Pistol:Exec
    Shotgun:Exec
    PaperBlueprint:Exec
    Rifle:Exec
    MeleeOneHand:Exec
    MachinePistol:Exec
    RocketLauncher:Exec
    GrenadeLauncher:Exec
    GoingCommando:Exec
    AssaultRifle:Exec
    TacticalShotgun:Exec
    SniperRifle:Exec
    TrapPlacement:Exec
    ShoulderLauncher:Exec
    AbilityDecoTool:Exec
    Crossbow:Exec
    C4:Exec
    RemoteControl:Exec
    DualWield:Exec
    AR_BullPup:Exec
    AR_ForwardGrip:Exec
    MedPackPaddles:Exec
    SMG_P90:Exec
    AR_DrumGun:Exec
    Consumable_Small:Exec
    Consumable_Large:Exec
    Balloon:Exec
    MountedTurret:Exec
    CreativeTool:Exec
    ExplosiveBow:Exec
    AshtonIndigo:Exec
    AshtonChicago:Exec
    MeleeDualWield:Exec
    Unarmed:Exec
   )
   > Set PickAxeOut(PickAxeOut=true)
   Set PistolOut(execute=@n14.Pistol, PistolOut=true)
   Event OnEndPickaxeSwing() {is=title K2Node_Event}
   > Set isSwingingPickaxe(isSwingingPickaxe=false)
  ]}
]
variables [
 ActivateExtraFixes {type=boolean}
 AlphaBackpack {type=boolean}
 AlphaFE {type=number}
 Backpack {type=AthenaBackpackItemDefinition, default=null}
 Current Backpack {type=AthenaPickaxeItemDefinition, default=null}
 Delta Time X {type=number}
 Gravity Override Params {type=GravityOverrideParamsStruct}
 GravityOverride {type=Vector}
 IsAimingDown {type=number}
 Iscelebration {type=boolean}
 IsCelebrationEmote {type=number}
 isCrouchingf {type=number}
 IsFloatingMode {type=number}
 isSwingingf {type=number}
 isSwingingPickaxe {type=boolean}
 PickAxeOut {type=boolean}
 PistolOut {type=boolean}
 SimlationSpace {type=AnimPhysSimSpaceType}
 SkyDivingTransitionOn {type=boolean}
 TacticalPrinceBackpack_ItemDef {type=AthenaBackpackItemDefinition, default=/BRCosmetics/Athena/Items/Cosmetics/Backpacks/Backpack_TacticalPrince}
 TacticalPrinceBackpack_SkelMesh {type=SkeletalMesh, default=/BRCosmetics/Accessories/FORT_Backpacks/Backpack_M_MED_TacticalPrince/Meshes/M_MED_TacticalPrince_Pack}
]
